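# Parent body material for the LazarusLens Style character cosmetic, adding glowing tattoos on top of the shared 2023 character shading.
# MF_CharacterParent_2023_PreFX builds the material attributes and supplies a Tech Art Mask, whose green channel drives the LazarusLens tattoo glow function.
# The glow is added to the existing emissive color before MF_CharacterParent_2023_PostFX processes the result.
# Shading model comes from the graph, with skeletal mesh, morph target, mesh deformer and Nanite usage enabled.
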Material /BRCosmetics/Characters/Player/Male/Medium/Bodies/M_MED_LazarusLens_Style/Materials/M_LazarusLens_Style_Body_Parent
graph {
 properties {
  bUseMaterialAttributes=true
  bUsedWithMeshDeformer=true
  bUsedWithMorphTargets=true
  bUsedWithNanite=true
  bUsedWithSkeletalMesh=true
  pixelDepthOffsetMode=PDOM_Legacy
  shadingModel=MSM_FromMaterialExpression
  usedShadingModels=Default Lit | Subsurface | Clear Coat | Subsurface Profile | Cloth
 }
 nodes [
  n1 = MaterialFunctionCall() {materialFunction=&MF_CharacterParent_2023_PreFX}
  n2 = GetMaterialAttributes(None=@n1.Result) {attributeGetTypes [#1]}
  n3 = MaterialFunctionCall!(TattooMask=ComponentMask(None=@n1.Tech Art Mask) {g=true}) {materialFunction=/Game/Characters/Player/Male/Medium/Bodies/M_MED_LazarusLens/Materials/MF_LazarusLens_TattooGlow}
  MaterialFunctionCall(
   Material=SetMaterialAttributes(MaterialAttributes=@n1.Result, Emissive Color=Add(A=@n2.Emissive Color, B=@n3.Result)) {attributeSetTypes [#1]}
   M=@n1.M
   Pre Skinned Local Position=@n1.M
   Distance Blend=@n1.M
   Fresnel=@n1.M
  ) {materialFunction=&MF_CharacterParent_2023_PostFX}
 ]
}
