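# Physics asset for the LazarusLens medium male player body, covering dangling jacket and cloth bones.
# Pelvis, spine_01, leg and Floor_COL bodies provide collision shapes around the garment.
# Limited swing/twist chains hang from small MSTR anchors: jacket sections on spine_01, back and side cloth on pelvis.
# Angularly free constraints cross-link neighbouring chains into connected panels; only the left jacket side links to the back.
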PhysicsAsset /BRCosmetics/Characters/Player/Male/Medium/Bodies/M_MED_LazarusLens_Style/Meshes/M_MED_LazarusLens_Style_Physics
bodies {
 Floor_COL [
  {shapeType=Sphere, radius=212, shapeName=floorCOL_sphere, center=<-3.4923 -2.714 -209>}
  {shapeType=Sphere, radius=27.2085, shapeName=floorCOL_sphere, center=<0 2.435 -0.4285>}
 ]
 calf_l [{shapeType=Capsule, length=29.5571, radius=10.2343, center=<-25 -2 0>, rotation=<rot 90 0 0>}]
 calf_r [{shapeType=Capsule, length=29.4321, radius=10.3593, center=<25 2 0.1223>, rotation=<rot 90 0 0>}]
 dyn_bk_cloth_01 [{shapeType=Box, extentX=2.2137, extentY=15.4596, extentZ=13.4563, shapeName=dyn_bk_cloth_01_box, center=<9 0 0>, rotation=<rot 0 -90 0>}]
 dyn_bk_cloth_02 [{shapeType=Box, extentX=3.8086, extentY=14.106, extentZ=16.2714, shapeName=dyn_bk_cloth_02_box, center=<7.3093 0 0>, rotation=<rot 0 -90 0>}]
 dyn_bk_cloth_03 [{shapeType=Capsule, length=18.0262, radius=6.5337, shapeName=dyn_bk_cloth_04_capsule, center=<8 -5 0>}]
 dyn_bk_cloth_04 [{shapeType=Capsule, length=19.6512, radius=6.5337, shapeName=dyn_bk_cloth_04_capsule, center=<6 -5 0>}]
 dyn_bk_cloth_MSTR [{shapeType=Capsule, length=0.505, radius=0.505, shapeName=dyn_bk_cloth_MSTR_capsule, rotation=<rot 0 -90 0>}]
 dyn_jacket_bk [{shapeType=Capsule, length=17.625, radius=8.125, center=<8 -5 0>}]
 dyn_jacket_bk_MSTR [{shapeType=Capsule, length=0.505, radius=0.505, shapeName=dyn_jacket_bk_MSTR_capsule, rotation=<rot 0 -90 0>}]
 dyn_jacket_bk_l_01 [{shapeType=Box, extentX=2.7605, extentY=8.7458, extentZ=15.6006, shapeName=dyn_jacket_bk_r_02_box, center=<3 0 2>, rotation=<rot 0 90 0>}]
 dyn_jacket_bk_l_02 [{
   shapeType=Capsule
   length=14.875
   radius=6.6875
   shapeName=dyn_jacket_bk_l_01_capsule
   center=<5 -4 0>
   rotation=<rot 0.0001 0.0003 -10.0001>
  }]
 dyn_jacket_bk_l_03 [{shapeType=Capsule, length=8.9375, radius=5.25, shapeName=dyn_jacket_bk_l_03_capsule, center=<6 -4 0>, rotation=<rot 90 0 0>}]
 dyn_jacket_bk_r_01 [{shapeType=Box, extentX=2.7605, extentY=8.7458, extentZ=15.6006, shapeName=dyn_jacket_bk_r_02_box, center=<3 0 -2>, rotation=<rot 0 90 0>}]
 dyn_jacket_bk_r_02 [{shapeType=Capsule, length=14.875, radius=6.6875, shapeName=dyn_jacket_bk_r_01_capsule, center=<5 -4 0>, rotation=<rot -0.0003 0.0009 5.0002>}]
 dyn_jacket_bk_r_03 [{shapeType=Capsule, length=8.9375, radius=5.25, shapeName=dyn_jacket_bk_r_03_capsule, center=<6 -4 0>, rotation=<rot 90 0 0>}]
 dyn_jacket_fr_l_01 [{shapeType=Capsule, length=8, radius=6.5, shapeName=dyn_jacket_fr_l_01_capsule, center=<6 -5 0>, rotation=<rot 90 0 0>}]
 dyn_jacket_fr_l_02 [{shapeType=Capsule, length=8, radius=6.5, shapeName=dyn_jacket_fr_l_02_capsule, center=<6 -5 0>, rotation=<rot 90 0 0>}]
 dyn_jacket_fr_l_03 [{shapeType=Capsule, length=8, radius=6, shapeName=dyn_jacket_fr_l_03_capsule, center=<6 -5 0>, rotation=<rot 90 0 0>}]
 dyn_jacket_fr_l_04 [{shapeType=Capsule, length=6.875, radius=6.375, center=<6 -5 0>, rotation=<rot 90 0 0>}]
 dyn_jacket_fr_l_05 [
  {shapeType=Capsule, length=6.5, radius=6.375, shapeName=dyn_jacket_fr_l_05_capsule, center=<8 -5 0>, rotation=<rot 90 0 0>}
  {shapeType=Capsule, length=5.125, radius=7.875, shapeName=dyn_jacket_fr_l_05_capsule, center=<4.0488 -4.9009 -9.3153>, rotation=<rot 90 0 0>}
 ]
 dyn_jacket_fr_l_06 [{shapeType=Capsule, length=14, radius=6.875, shapeName=dyn_jacket_fr_l_06_capsule, center=<9 -5 0>, rotation=<rot 90 0 0>}]
 dyn_jacket_fr_l_MSTR [{shapeType=Capsule, length=0.505, radius=0.505, shapeName=dyn_jacket_fr_l_MSTR_capsule, rotation=<rot 0 -90 0>}]
 dyn_jacket_fr_r_01 [{shapeType=Capsule, length=8, radius=6.5, shapeName=dyn_jacket_fr_l_01_capsule, center=<6 -5 0>, rotation=<rot 90 0 0>}]
 dyn_jacket_fr_r_02 [{shapeType=Capsule, length=8, radius=6.5, shapeName=dyn_jacket_fr_l_02_capsule, center=<6 -5 0>, rotation=<rot 90 0 0>}]
 dyn_jacket_fr_r_03 [{shapeType=Capsule, length=8, radius=6, shapeName=dyn_jacket_fr_l_03_capsule, center=<6 -5 0>, rotation=<rot 90 0 0>}]
 dyn_jacket_fr_r_04 [{shapeType=Capsule, length=6.125, radius=6.375, center=<6 -5 0>, rotation=<rot 90 0 0>}]
 dyn_jacket_fr_r_05 [
  {shapeType=Capsule, length=5.75, radius=6.375, shapeName=dyn_jacket_fr_l_05_capsule, center=<8 -5 0>, rotation=<rot 90 0 0>}
  {shapeType=Capsule, length=5.75, radius=7.25, shapeName=dyn_jacket_fr_r_05_capsule, center=<2.1905 -5.1603 11.3977>, rotation=<rot 90 0 0>}
 ]
 dyn_jacket_fr_r_06 [{shapeType=Capsule, length=13.375, radius=6.875, shapeName=dyn_jacket_fr_l_06_capsule, center=<9 -5 0>, rotation=<rot 90 0 0>}]
 dyn_jacket_fr_r_MSTR [{shapeType=Capsule, length=0.505, radius=0.505, shapeName=dyn_jacket_fr_r_MSTR_capsule, rotation=<rot 0 -90 0>}]
 dyn_jacket_l_01 [{shapeType=Capsule, length=10.6875, radius=9.6875, shapeName=dyn_jacket_l_01_capsule, center=<6 -7 0>, rotation=<rot 0.0001 0 -5>}]
 dyn_jacket_l_02 [{shapeType=Capsule, length=9.9375, radius=10.375, shapeName=dyn_jacket_l_02_capsule, center=<7 -7 0>, rotation=<rot 0.0001 0 -10>}]
 dyn_jacket_l_03 [{shapeType=Capsule, length=13, radius=9.5625, center=<8 -6 0>}]
 dyn_jacket_l_04 [
  {shapeType=Capsule, length=12.125, radius=10.4375, shapeName=dyn_jacket_l_04_capsule, center=<8 -5 0>}
  {
   shapeType=Capsule
   length=10.75
   radius=6.5
   shapeName=dyn_jacket_l_04_capsule
   center=<20.341 -1.2748 4.5175>
   rotation=<rot -24.9998 0 19.9999>
  }
 ]
 dyn_jacket_l_MSTR [{shapeType=Capsule, length=0.505, radius=0.505, shapeName=dyn_jacket_l_MSTR_capsule, rotation=<rot 0 -90 0>}]
 dyn_jacket_r_01 [{shapeType=Capsule, length=10.6875, radius=9.6875, shapeName=dyn_jacket_l_01_capsule, center=<6 -7 0>}]
 dyn_jacket_r_02 [{shapeType=Capsule, length=8.5625, radius=10.375, shapeName=dyn_jacket_l_02_capsule, center=<7 -7 0>, rotation=<rot 0 0 5>}]
 dyn_jacket_r_03 [{shapeType=Capsule, length=13, radius=9.5625, center=<8 -6 0>}]
 dyn_jacket_r_04 [
  {shapeType=Capsule, length=12.125, radius=10.4375, shapeName=dyn_jacket_l_04_capsule, center=<8 -5 0>}
  {
   shapeType=Capsule
   length=12
   radius=5.75
   shapeName=dyn_jacket_r_04_capsule
   center=<15.9635 -0.7743 -4.8643>
   rotation=<rot 43.4757 -6.8716 -34.3616>
  }
 ]
 dyn_jacket_r_MSTR [{shapeType=Capsule, length=0.505, radius=0.505, shapeName=dyn_jacket_r_MSTR_capsule, rotation=<rot 0 -90 0>}]
 dyn_l_cloth_01 [{shapeType=Box, extentX=3.5887, extentY=11.3346, extentZ=9.0813, shapeName=dyn_bk_cloth_01_box, center=<6 1.2 0>, rotation=<rot 0 -90 0>}]
 dyn_l_cloth_02 [{shapeType=Box, extentX=3.4336, extentY=14.106, extentZ=9.6464, shapeName=dyn_bk_cloth_02_box, center=<7.3093 1.2 0>, rotation=<rot 0 -90 0>}]
 dyn_l_cloth_03 [{shapeType=Capsule, length=10.2082, radius=7.6309, shapeName=dyn_r_cloth_04_capsule, center=<7 -6 -3>}]
 dyn_l_cloth_04 [{shapeType=Capsule, length=12.9582, radius=3.3809, shapeName=dyn_r_cloth_04_capsule, center=<2 -2 -3>}]
 dyn_l_cloth_MSTR [{shapeType=Capsule, length=0.505, radius=0.505, shapeName=dyn_l_cloth_MSTR_capsule, rotation=<rot 0 -90 0>}]
 dyn_r_cloth_01 [{shapeType=Box, extentX=3.5887, extentY=11.3346, extentZ=9.0813, shapeName=dyn_bk_cloth_01_box, center=<6 1.2 0>, rotation=<rot 0 -90 0>}]
 dyn_r_cloth_02 [{shapeType=Box, extentX=3.4336, extentY=14.106, extentZ=9.6464, shapeName=dyn_bk_cloth_02_box, center=<7.3093 1.2 0>, rotation=<rot 0 -90 0>}]
 dyn_r_cloth_03 [{shapeType=Capsule, length=10.2082, radius=7.6309, shapeName=dyn_r_cloth_04_capsule, center=<7 -6 3>}]
 dyn_r_cloth_04 [{shapeType=Capsule, length=12.9582, radius=3.3809, shapeName=dyn_r_cloth_04_capsule, center=<2 -2 3>}]
 dyn_r_cloth_MSTR [{shapeType=Capsule, length=0.505, radius=0.505, shapeName=dyn_r_cloth_MSTR_capsule, rotation=<rot 0 -90 0>}]
 foot_l [{shapeType=Capsule, length=13.0862, radius=9.4382, center=<-3.397 6.7099 0.0834>, rotation=<rot 0.7349 0.9585 -90.8278>}]
 foot_r [{shapeType=Capsule, length=17.1001, radius=9.455, center=<3.487 -8.59 -2.0824>, rotation=<rot 0.7221 1.1908 -95.9009>}]
 pelvis [
  {shapeType=Capsule, length=10.2374, radius=9.6205, center=<-3.1833 -1.5286 0>}
  {shapeType=Capsule, length=10.5, radius=9.5625, shapeName=pelvis_capsule, center=<-5.1189 -3.6977 0>}
 ]
 spine_01 [{shapeType=Capsule, length=6.4874, radius=9.433, center=<-3.1833 -1.2489 0>}]
 thigh_l [{
   shapeType=Capsule
   length=30.7226
   radius=9.3577
   shapeName=thigh_l_capsule
   center=<-23.9168 -1.8187 0.7331>
   rotation=<rot 87.8125 -178.3106 179.9997>
  }]
 thigh_r [{
   shapeType=Capsule
   length=30.7226
   radius=9.3577
   shapeName=thigh_r_capsule
   center=<23.93 1.3712 -0.5163>
   rotation=<rot 87.8125 -178.3105 179.9998>
  }]
}
constraints [
 {bone1Name=spine_01, bone2Name=pelvis, swing1LimitDegrees=45, swing2LimitDegrees=45, twistLimitDegrees=45}
 {bone1Name=dyn_jacket_fr_l_04, bone2Name=dyn_jacket_fr_l_03, swing1LimitDegrees=35, swing2LimitDegrees=35, twistLimitDegrees=15}
 {bone1Name=dyn_jacket_r_03, bone2Name=dyn_jacket_r_02, swing1LimitDegrees=35, swing2LimitDegrees=35, twistLimitDegrees=15}
 {bone1Name=dyn_jacket_l_03, bone2Name=dyn_jacket_l_02, swing1LimitDegrees=35, swing2LimitDegrees=35, twistLimitDegrees=15}
 {bone1Name=dyn_jacket_bk, bone2Name=dyn_jacket_bk_MSTR, swing1LimitDegrees=35, swing2LimitDegrees=35, twistLimitDegrees=15}
 {bone1Name=thigh_l, bone2Name=pelvis, swing1LimitDegrees=45, swing2LimitDegrees=45, twistLimitDegrees=45}
 {bone1Name=calf_l, bone2Name=thigh_l, swing1LimitDegrees=45, swing2LimitDegrees=45, twistLimitDegrees=45}
 {bone1Name=thigh_r, bone2Name=pelvis, swing1LimitDegrees=45, swing2LimitDegrees=45, twistLimitDegrees=45}
 {bone1Name=calf_r, bone2Name=thigh_r, swing1LimitDegrees=45, swing2LimitDegrees=45, twistLimitDegrees=45}
 {bone1Name=dyn_bk_cloth_03, bone2Name=dyn_bk_cloth_02, swing1LimitDegrees=70, swing2LimitDegrees=35, twistLimitDegrees=20}
 {
  bone1Name=foot_r
  bone2Name=calf_r
  swing1LimitDegrees=45
  swing1Motion=Free
  swing2LimitDegrees=45
  swing2Motion=Free
  twistLimitDegrees=45
  twistMotion=Free
 }
 {
  bone1Name=foot_l
  bone2Name=calf_l
  swing1LimitDegrees=45
  swing1Motion=Free
  swing2LimitDegrees=45
  swing2Motion=Free
  twistLimitDegrees=45
  twistMotion=Free
 }
 {bone1Name=dyn_jacket_fr_r_06, bone2Name=dyn_jacket_fr_r_05, swing1LimitDegrees=35, swing2LimitDegrees=35, twistLimitDegrees=15}
 {
  bone1Name=dyn_jacket_fr_r_MSTR
  bone2Name=spine_01
  swing1LimitDegrees=45
  swing1Motion=Free
  swing2LimitDegrees=45
  swing2Motion=Free
  twistLimitDegrees=45
  twistMotion=Free
 }
 {bone1Name=dyn_jacket_fr_r_01, bone2Name=dyn_jacket_fr_r_MSTR, swing1LimitDegrees=35, swing2LimitDegrees=35, twistLimitDegrees=5}
 {bone1Name=dyn_jacket_fr_r_02, bone2Name=dyn_jacket_fr_r_01, swing1LimitDegrees=35, swing2LimitDegrees=35, twistLimitDegrees=5}
 {bone1Name=dyn_jacket_fr_r_03, bone2Name=dyn_jacket_fr_r_02, swing1LimitDegrees=35, swing2LimitDegrees=35, twistLimitDegrees=5}
 {bone1Name=dyn_jacket_fr_r_04, bone2Name=dyn_jacket_fr_r_03, swing1LimitDegrees=35, swing2LimitDegrees=35, twistLimitDegrees=15}
 {bone1Name=dyn_jacket_fr_r_05, bone2Name=dyn_jacket_fr_r_04, swing1LimitDegrees=35, swing2LimitDegrees=35, twistLimitDegrees=15}
 {bone1Name=dyn_jacket_fr_l_06, bone2Name=dyn_jacket_fr_l_05, swing1LimitDegrees=35, swing2LimitDegrees=35, twistLimitDegrees=15}
 {bone1Name=dyn_jacket_fr_l_05, bone2Name=dyn_jacket_fr_l_04, swing1LimitDegrees=35, swing2LimitDegrees=35, twistLimitDegrees=15}
 {bone1Name=dyn_jacket_fr_l_03, bone2Name=dyn_jacket_fr_l_02, swing1LimitDegrees=35, swing2LimitDegrees=35, twistLimitDegrees=5}
 {bone1Name=dyn_jacket_fr_l_02, bone2Name=dyn_jacket_fr_l_01, swing1LimitDegrees=35, swing2LimitDegrees=35, twistLimitDegrees=5}
 {bone1Name=dyn_jacket_fr_l_01, bone2Name=dyn_jacket_fr_l_MSTR, swing1LimitDegrees=35, swing2LimitDegrees=35, twistLimitDegrees=5}
 {
  bone1Name=dyn_jacket_fr_l_MSTR
  bone2Name=spine_01
  swing1LimitDegrees=45
  swing1Motion=Free
  swing2LimitDegrees=45
  swing2Motion=Free
  twistLimitDegrees=45
  twistMotion=Free
 }
 {bone1Name=dyn_jacket_bk_r_03, bone2Name=dyn_jacket_bk_r_02, swing1LimitDegrees=35, swing2LimitDegrees=35, twistLimitDegrees=5}
 {bone1Name=dyn_jacket_bk_l_01, bone2Name=dyn_jacket_bk, swing1LimitDegrees=35, swing2LimitDegrees=35, twistLimitDegrees=5}
 {bone1Name=dyn_jacket_bk_l_02, bone2Name=dyn_jacket_bk_l_01, swing1LimitDegrees=35, swing2LimitDegrees=35, twistLimitDegrees=5}
 {bone1Name=dyn_jacket_bk_l_03, bone2Name=dyn_jacket_bk_l_02, swing1LimitDegrees=35, swing2LimitDegrees=35, twistLimitDegrees=5}
 {bone1Name=dyn_jacket_bk_r_01, bone2Name=dyn_jacket_bk, swing1LimitDegrees=35, swing2LimitDegrees=35, twistLimitDegrees=5}
 {bone1Name=dyn_jacket_bk_r_02, bone2Name=dyn_jacket_bk_r_01, swing1LimitDegrees=35, swing2LimitDegrees=35, twistLimitDegrees=5}
 {
  bone1Name=dyn_jacket_bk_MSTR
  bone2Name=spine_01
  swing1LimitDegrees=45
  swing1Motion=Free
  swing2LimitDegrees=45
  swing2Motion=Free
  twistLimitDegrees=45
  twistMotion=Free
 }
 {bone1Name=dyn_r_cloth_04, bone2Name=dyn_r_cloth_03, swing1LimitDegrees=60, swing2LimitDegrees=35, twistLimitDegrees=20}
 {
  bone1Name=dyn_l_cloth_MSTR
  bone2Name=pelvis
  swing1LimitDegrees=45
  swing1Motion=Free
  swing2LimitDegrees=45
  swing2Motion=Free
  twistLimitDegrees=45
  twistMotion=Free
 }
 {bone1Name=dyn_l_cloth_01, bone2Name=dyn_l_cloth_MSTR, swing1LimitDegrees=60, swing2LimitDegrees=35, twistLimitDegrees=15}
 {bone1Name=dyn_l_cloth_02, bone2Name=dyn_l_cloth_01, swing1LimitDegrees=60, swing2LimitDegrees=35, twistLimitDegrees=15}
 {bone1Name=dyn_l_cloth_03, bone2Name=dyn_l_cloth_02, swing1LimitDegrees=60, swing2LimitDegrees=35, twistLimitDegrees=20}
 {bone1Name=dyn_l_cloth_04, bone2Name=dyn_l_cloth_03, swing1LimitDegrees=60, swing2LimitDegrees=35, twistLimitDegrees=20}
 {
  bone1Name=dyn_r_cloth_MSTR
  bone2Name=pelvis
  swing1LimitDegrees=45
  swing1Motion=Free
  swing2LimitDegrees=45
  swing2Motion=Free
  twistLimitDegrees=45
  twistMotion=Free
 }
 {bone1Name=dyn_r_cloth_01, bone2Name=dyn_r_cloth_MSTR, swing1LimitDegrees=60, swing2LimitDegrees=35, twistLimitDegrees=15}
 {bone1Name=dyn_r_cloth_02, bone2Name=dyn_r_cloth_01, swing1LimitDegrees=60, swing2LimitDegrees=35, twistLimitDegrees=15}
 {bone1Name=dyn_r_cloth_03, bone2Name=dyn_r_cloth_02, swing1LimitDegrees=60, swing2LimitDegrees=35, twistLimitDegrees=20}
 {bone1Name=dyn_bk_cloth_04, bone2Name=dyn_bk_cloth_03, swing1LimitDegrees=70, swing2LimitDegrees=35, twistLimitDegrees=20}
 {bone1Name=dyn_bk_cloth_02, bone2Name=dyn_bk_cloth_01, swing1LimitDegrees=70, swing2LimitDegrees=35, twistLimitDegrees=15}
 {bone1Name=dyn_bk_cloth_01, bone2Name=dyn_bk_cloth_MSTR, swing1LimitDegrees=70, swing2LimitDegrees=35, twistLimitDegrees=15}
 {
  bone1Name=dyn_bk_cloth_MSTR
  bone2Name=pelvis
  swing1LimitDegrees=45
  swing1Motion=Free
  swing2LimitDegrees=45
  swing2Motion=Free
  twistLimitDegrees=45
  twistMotion=Free
 }
 {
  bone1Name=dyn_jacket_l_MSTR
  bone2Name=spine_01
  swing1LimitDegrees=45
  swing1Motion=Free
  swing2LimitDegrees=45
  swing2Motion=Free
  twistLimitDegrees=45
  twistMotion=Free
 }
 {bone1Name=dyn_jacket_l_01, bone2Name=dyn_jacket_l_MSTR, swing1LimitDegrees=35, swing2LimitDegrees=35, twistLimitDegrees=15}
 {bone1Name=dyn_jacket_l_02, bone2Name=dyn_jacket_l_01, swing1LimitDegrees=35, swing2LimitDegrees=35, twistLimitDegrees=15}
 {bone1Name=dyn_jacket_l_04, bone2Name=dyn_jacket_l_03, swing1LimitDegrees=35, swing2LimitDegrees=35, twistLimitDegrees=15}
 {
  bone1Name=dyn_jacket_r_MSTR
  bone2Name=spine_01
  swing1LimitDegrees=45
  swing1Motion=Free
  swing2LimitDegrees=45
  swing2Motion=Free
  twistLimitDegrees=45
  twistMotion=Free
 }
 {bone1Name=dyn_jacket_r_01, bone2Name=dyn_jacket_r_MSTR, swing1LimitDegrees=35, swing2LimitDegrees=35, twistLimitDegrees=15}
 {bone1Name=dyn_jacket_r_02, bone2Name=dyn_jacket_r_01, swing1LimitDegrees=35, swing2LimitDegrees=35, twistLimitDegrees=15}
 {bone1Name=dyn_jacket_r_04, bone2Name=dyn_jacket_r_03, swing1LimitDegrees=35, swing2LimitDegrees=35, twistLimitDegrees=15}
 {
  bone1Name=dyn_r_cloth_03
  bone2Name=dyn_bk_cloth_03
  swing1LimitDegrees=45
  swing1Motion=Free
  swing2LimitDegrees=45
  swing2Motion=Free
  twistLimitDegrees=45
  twistMotion=Free
 }
 {
  bone1Name=dyn_l_cloth_03
  bone2Name=dyn_bk_cloth_03
  swing1LimitDegrees=45
  swing1Motion=Free
  swing2LimitDegrees=45
  swing2Motion=Free
  twistLimitDegrees=45
  twistMotion=Free
 }
 {
  bone1Name=dyn_r_cloth_04
  bone2Name=dyn_bk_cloth_04
  swing1LimitDegrees=45
  swing1Motion=Free
  swing2LimitDegrees=45
  swing2Motion=Free
  twistLimitDegrees=45
  twistMotion=Free
 }
 {
  bone1Name=dyn_l_cloth_04
  bone2Name=dyn_bk_cloth_04
  swing1LimitDegrees=45
  swing1Motion=Free
  swing2LimitDegrees=45
  swing2Motion=Free
  twistLimitDegrees=45
  twistMotion=Free
 }
 {
  bone1Name=dyn_jacket_fr_l_04
  bone2Name=dyn_jacket_l_04
  swing1LimitDegrees=45
  swing1Motion=Free
  swing2LimitDegrees=45
  swing2Motion=Free
  twistLimitDegrees=45
  twistMotion=Free
 }
 {
  bone1Name=dyn_jacket_fr_l_02
  bone2Name=dyn_jacket_l_02
  swing1LimitDegrees=45
  swing1Motion=Free
  swing2LimitDegrees=45
  swing2Motion=Free
  twistLimitDegrees=45
  twistMotion=Free
 }
 {
  bone1Name=dyn_jacket_fr_r_02
  bone2Name=dyn_jacket_r_02
  swing1LimitDegrees=45
  swing1Motion=Free
  swing2LimitDegrees=45
  swing2Motion=Free
  twistLimitDegrees=45
  twistMotion=Free
 }
 {
  bone1Name=dyn_jacket_l_03
  bone2Name=dyn_jacket_bk_l_02
  swing1LimitDegrees=45
  swing1Motion=Free
  swing2LimitDegrees=45
  swing2Motion=Free
  twistLimitDegrees=45
  twistMotion=Free
 }
 {
  bone1Name=dyn_jacket_fr_r_03
  bone2Name=dyn_jacket_r_03
  swing1LimitDegrees=45
  swing1Motion=Free
  swing2LimitDegrees=45
  swing2Motion=Free
  twistLimitDegrees=45
  twistMotion=Free
 }
 {
  bone1Name=dyn_jacket_fr_l_03
  bone2Name=dyn_jacket_l_03
  swing1LimitDegrees=45
  swing1Motion=Free
  swing2LimitDegrees=45
  swing2Motion=Free
  twistLimitDegrees=45
  twistMotion=Free
 }
 {
  bone1Name=dyn_jacket_fr_r_04
  bone2Name=dyn_jacket_r_04
  swing1LimitDegrees=45
  swing1Motion=Free
  swing2LimitDegrees=45
  swing2Motion=Free
  twistLimitDegrees=45
  twistMotion=Free
 }
]
