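# Animation Blueprint for the LazarusLens Style male body cosmetic, adding secondary motion to jacket, cloth, skirt, necklace and wristbands on a pose copied from the parent mesh.
# RigidBody simulations use separate physics assets for gameplay, skydiving and front end, blended by state.
# AnimDynamics and weighted multi-bone constraints then drive the accessories.
# Variable-driven bone offsets and leg constraints adjust them for movement, limb rotation, backbling, crouch and skydive.
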AnimBlueprint /BRCosmetics/Characters/Player/Male/Medium/Bodies/M_MED_LazarusLens_Style/Meshes/M_MED_LazarusLens_Style_AnimBP
target_skeleton=&/M_MED_LazarusLens_Style_Skeleton
graphs [
 AnimGraph {nodes [
   n1 = "Constraint - dyn_l_cloth_01 (R)"(Alpha=1, ConstraintWeights_0=1) {
    alphaBoolBlend {blendOption=Linear}
    boneToModify {boneName=dyn_l_cloth_01}
    constraintSetup [{transformType=Rotation, targetBone {boneName=thigh_l}}]
    constraintWeights [1]
   }
   n2 = "Constraint - dyn_l_cloth_03 (R)"(ComponentPose=@n1, Alpha=1, ConstraintWeights_0=1) {
    alphaBoolBlend {blendOption=Linear}
    boneToModify {boneName=dyn_l_cloth_03}
    constraintSetup [{transformType=Rotation, targetBone {boneName=calf_l}}]
    constraintWeights [1]
   }
   n3 = "Constraint - dyn_r_cloth_01 (R)"(ComponentPose=@n2, Alpha=1, ConstraintWeights_0=1) {
    alphaBoolBlend {blendOption=Linear}
    boneToModify {boneName=dyn_r_cloth_01}
    constraintSetup [{transformType=Rotation, targetBone {boneName=thigh_r}}]
    constraintWeights [1]
   }
   n4 = "Constraint - dyn_r_cloth_03 (R)"(ComponentPose=@n3, Alpha=1, ConstraintWeights_0=1) {
    alphaBoolBlend {blendOption=Linear}
    boneToModify {boneName=dyn_r_cloth_03}
    constraintSetup [{transformType=Rotation, targetBone {boneName=calf_r}}]
    constraintWeights [1]
   }
   n5 = "Constraint - dyn_bk_cloth_01 (RR)"!(ComponentPose=@n4, Alpha=1, ConstraintWeights_0=0.5, ConstraintWeights_1:float=0.5) -> (Pose) {
    alphaBoolBlend {blendOption=Linear}
    boneToModify {boneName=dyn_bk_cloth_01}
    constraintSetup [{transformType=Rotation, targetBone {boneName=thigh_r}}, {transformType=Rotation, targetBone {boneName=thigh_l}}]
    constraintWeights [1,1]
   }
   "Constraint - dyn_bk_cloth_03 (RR)"!(ComponentPose=@n5, Alpha=1, ConstraintWeights_0=0.5, ConstraintWeights_1:float=0.5) -> (Pose) {
    alphaBoolBlend {blendOption=Linear}
    boneToModify {boneName=dyn_bk_cloth_03}
    constraintSetup [{transformType=Rotation, targetBone {boneName=calf_r}}, {transformType=Rotation, targetBone {boneName=calf_l}}]
    constraintWeights [1,1]
   }
   n6 = Get moving2d() -> (moving2d:double)
   "Transform (Modify) Bone - Bone: dyn_jacket_bk_MSTR"(Alpha=@n6, Translation="0,-2,0", Rotation="0,-5,0", Scale="1,1,1") {
    rotationMode=BMM_Additive
    rotationSpace=BCS_BoneSpace
    translationMode=BMM_Additive
    alphaBoolBlend {blendOption=Linear}
    alphaScaleBiasClamp {bInterpResult=true}
    boneToModify {boneName=dyn_jacket_bk_MSTR}
    rotation {yaw=-5}
    translation {y=-2}
   }
   RigidBody!(ComponentPose, bAlphaBoolEnabled:bool, OverrideWorldGravity:Vector="0,0,-980", ExternalForce) -> (Pose) {
    alphaInputType=Bool
    bClampLinearTranslationLimitToRefPose=true
    bOverrideWorldGravity=true
    overridePhysicsAsset=&/M_MED_LazarusLens_Style_Physics
    simulationSpace=BaseBoneSpace
    alphaBoolBlend {blendInTime=0.1, blendOption=Linear, blendOutTime=0.1}
    baseBoneRef {boneName=pelvis}
    componentAppliedLinearAccClamp=<3000 3000 3000>
    componentLinearAccScale=<0.3 0.3 0.5>
    componentLinearVelScale=<0.3 0.3 0.5>
    simSpaceSettings {
     maxAngularAcceleration=1000
     maxAngularVelocity=3
     maxLinearAcceleration=700
     maxLinearVelocity=500
     velocityScaleZ=0
     externalLinearDragV=<0.2 0.2>
    }
   }
   n7 = Get Gravity Override()
   n8 = Get bIsSkyDiving()
   n9 = RigidBody(ComponentPose=Local To Component(LocalPose=Use cached pose 'pre rigid'()), bAlphaBoolEnabled=@n8, OverrideWorldGravity=@n7) {
    alphaInputType=Bool
    bClampLinearTranslationLimitToRefPose=true
    bOverrideWorldGravity=true
    overridePhysicsAsset=&/M_MED_LazarusLens_Style_skydive_Physics
    simulationSpace=BaseBoneSpace
    alphaBoolBlend {blendInTime=0.1, blendOption=Linear, blendOutTime=0.1}
    baseBoneRef {boneName=pelvis}
    componentAppliedLinearAccClamp=<3000 3000 3000>
    componentLinearAccScale=<0.3 0.3 0.3>
    componentLinearVelScale=<0.3 0.3 0.3>
   }
   n10 = RigidBody!(
    ComponentPose=@n9
    bAlphaBoolEnabled=NOT Boolean(A=@n8)
    OverrideWorldGravity=@n7
    ExternalForce
    SimSpaceSettings:Sim Space Settings Structure=Get Sim Space Settings() -> (Sim Space Settings:Sim Space Settings Structure)
   ) -> (Pose) {
    alphaInputType=Bool
    bClampLinearTranslationLimitToRefPose=true
    bOverrideWorldGravity=true
    overridePhysicsAsset=&M_MED_LazarusLens_Style_Physics
    simulationSpace=BaseBoneSpace
    alphaBoolBlend {blendInTime=0.1, blendOption=Linear, blendOutTime=0.1}
    baseBoneRef {boneName=pelvis}
    componentAppliedLinearAccClamp=<3000 3000 3000>
    componentLinearAccScale=<0.3 0.3 0.5>
    componentLinearVelScale=<0.3 0.3 0.5>
    simSpaceSettings {
     maxAngularAcceleration=1000
     maxAngularVelocity=3
     maxLinearAcceleration=700
     maxLinearVelocity=500
     velocityScaleZ=0
     externalLinearDragV=<0.2 0.2>
    }
   }
   n11 = Component To Local(ComponentPose=@n10)
   n12 = RigidBody!(ComponentPose=Local To Component(LocalPose=Use cached pose 'pre rigid'()), bAlphaBoolEnabled, OverrideWorldGravity=@n7, ExternalForce) -> (Pose) {
    alphaInputType=Bool
    bClampLinearTranslationLimitToRefPose=true
    bOverrideWorldGravity=true
    overridePhysicsAsset=&/M_MED_LazarusLens_Style_FE_Physics
    simulationSpace=BaseBoneSpace
    alphaBoolBlend {blendInTime=0.1, blendOption=Linear, blendOutTime=0.1}
    baseBoneRef {boneName=root}
    componentAppliedLinearAccClamp=<3000 3000 3000>
    componentLinearAccScale=<0.3 0.3 0.3>
    componentLinearVelScale=<0.3 0.3 0.3>
   }
   n13 = Component To Local(ComponentPose=@n12)
   n14 = Get bIsInFrontEnd()
   n15 = Anim Dynamics - Bone: dyn_necklace_A(
    ComponentPose=Local To Component(LocalPose=Blend Poses by bool(BlendPose_0=@n13, BlendPose_1=@n11, bActiveValue=@n14) {blendType=Linear})
    Alpha=1
    ExternalForce="0,0,0"
   ) {
    angularDampingOverride=0.999
    angularSpringConstant=70
    bAngularSpring=true
    bGravityOverrideInSimSpace=true
    bOverrideAngularDamping=true
    bUseGravityOverride=true
    bUseSphericalLimits=true
    simulationSpace=RootRelative
    alphaBoolBlend {blendOption=Linear}
    boundBone {boneName=dyn_necklace_A}
    componentAppliedLinearAccClamp=<1500 1500 1500>
    componentLinearAccScale=<0.3 0.3 0.3>
    componentLinearVelScale=<0.3 0.3 0.3>
    gravityOverride {z=-980}
    physicsBodyDefinitions [{
      collisionType=InnerSphere
      sphereCollisionRadius=1
      boundBone {boneName=dyn_necklace_A}
      boxExtents=<15 15 20>
      constraintSetup {twistAxis=AxisY, angularLimitsMax=<1 2 0>, angularLimitsMin=<-1 -2 -35>, angularTarget=<1 0 0>}
      localJointOffset=<1 0 0>
     }]
    planarLimits [{drivingBone {boneName=spine_05}, planeTransform {location=<-0.5 0 0>, rotation=<rot 0 -51 90>}}]
    relativeSpaceBone {boneName=root}
   }
   n16 = Anim Dynamics - Bone: dyn_necklace_B(ComponentPose=@n15, Alpha=1, ExternalForce="0,0,0") {
    angularDampingOverride=0.999
    angularSpringConstant=70
    bAngularSpring=true
    bGravityOverrideInSimSpace=true
    bOverrideAngularDamping=true
    bUseGravityOverride=true
    bUsePlanarLimit=false
    bUseSphericalLimits=true
    simulationSpace=RootRelative
    alphaBoolBlend {blendOption=Linear}
    boundBone {boneName=dyn_necklace_B}
    componentAppliedLinearAccClamp=<1500 1500 1500>
    componentLinearAccScale=<0.5 0.5 0.5>
    componentLinearVelScale=<0.5 0.5 0.5>
    gravityOverride {z=-980}
    physicsBodyDefinitions [{
      collisionType=InnerSphere
      sphereCollisionRadius=1
      boundBone {boneName=dyn_necklace_B}
      boxExtents=<12 15 15>
      constraintSetup {twistAxis=AxisY, angularLimitsMax=<1 3 0>, angularLimitsMin=<-1 -3 -35>, angularTarget=<1 0 0>}
      localJointOffset=<2 0 0>
     }]
    relativeSpaceBone {boneName=root}
   }
   n17 = Anim Dynamics - Bone: dyn_necklace_C(ComponentPose=@n16, Alpha=1, ExternalForce="0,0,0") {
    angularDampingOverride=0.999
    angularSpringConstant=70
    bAngularSpring=true
    bGravityOverrideInSimSpace=true
    bOverrideAngularDamping=true
    bUseGravityOverride=true
    bUsePlanarLimit=false
    bUseSphericalLimits=true
    simulationSpace=RootRelative
    alphaBoolBlend {blendOption=Linear}
    boundBone {boneName=dyn_necklace_C}
    componentAppliedLinearAccClamp=<1500 1500 1500>
    componentLinearAccScale=<0.5 0.5 0.5>
    componentLinearVelScale=<0.5 0.5 0.5>
    gravityOverride {z=-980}
    physicsBodyDefinitions [{
      collisionType=InnerSphere
      sphereCollisionRadius=1
      boundBone {boneName=dyn_necklace_C}
      boxExtents=<7 12 12>
      constraintSetup {twistAxis=AxisY, angularLimitsMax=<1 5 0>, angularLimitsMin=<-1 -5 -35>, angularTarget=<1 0 0>}
      localJointOffset=<2 0 0>
     }]
    relativeSpaceBone {boneName=root}
   }
   n18 = Anim Dynamics - Bone: dyn_necklace_C_01(ComponentPose=@n17, Alpha=1, ExternalForce="0,0,0") {
    angularDampingOverride=0.999
    angularSpringConstant=100
    bAngularSpring=true
    bGravityOverrideInSimSpace=true
    bOverrideAngularDamping=true
    bUseGravityOverride=true
    bUsePlanarLimit=false
    bUseSphericalLimits=true
    simulationSpace=RootRelative
    alphaBoolBlend {blendOption=Linear}
    boundBone {boneName=dyn_necklace_C_01}
    componentAppliedLinearAccClamp=<1500 1500 1500>
    componentLinearAccScale=<0.5 0.5 0.5>
    componentLinearVelScale=<0.5 0.5 0.5>
    gravityOverride {z=-980}
    physicsBodyDefinitions [{
      collisionType=InnerSphere
      sphereCollisionRadius=1
      boundBone {boneName=dyn_necklace_C_01}
      boxExtents=<10 5 5>
      constraintSetup {twistAxis=AxisY, angularLimitsMax=<1 5 0>, angularLimitsMin=<-1 -5 -35>, angularTarget=<1 0 0>}
      localJointOffset=<2 0 0>
     }]
    relativeSpaceBone {boneName=root}
    sphericalLimits [{limitRadius=14.75, drivingBone {boneName=spine_04}, sphereLocalOffset=<-1 -4 6>}]
   }
   n19 = Anim Dynamics - Bone: dyn_necklace_C_02(ComponentPose=@n18, Alpha=1, ExternalForce="0,0,0") {
    angularDampingOverride=0.999
    angularSpringConstant=100
    bAngularSpring=true
    bGravityOverrideInSimSpace=true
    bOverrideAngularDamping=true
    bUseGravityOverride=true
    bUsePlanarLimit=false
    bUseSphericalLimits=true
    simulationSpace=RootRelative
    alphaBoolBlend {blendOption=Linear}
    boundBone {boneName=dyn_necklace_C_02}
    componentAppliedLinearAccClamp=<1500 1500 1500>
    componentLinearAccScale=<0.5 0.5 0.5>
    componentLinearVelScale=<0.5 0.5 0.5>
    gravityOverride {z=-980}
    physicsBodyDefinitions [{
      collisionType=InnerSphere
      sphereCollisionRadius=1
      boundBone {boneName=dyn_necklace_C_02}
      boxExtents=<10 5 5>
      constraintSetup {twistAxis=AxisY, angularLimitsMax=<1 5 0>, angularLimitsMin=<-1 -5 -35>, angularTarget=<1 0 0>}
      localJointOffset=<2 0 0>
     }]
    relativeSpaceBone {boneName=root}
    sphericalLimits [{limitRadius=14.5, drivingBone {boneName=spine_03}, sphereLocalOffset=<2 -4 6>}]
   }
   n20 = Anim Dynamics - Bone: dyn_necklace_C_03(ComponentPose=@n19, Alpha=1, ExternalForce="0,0,0") {
    angularDampingOverride=0.999
    angularSpringConstant=100
    bGravityOverrideInSimSpace=true
    bOverrideAngularDamping=true
    bUseGravityOverride=true
    bUsePlanarLimit=false
    bUseSphericalLimits=true
    simulationSpace=RootRelative
    alphaBoolBlend {blendOption=Linear}
    boundBone {boneName=dyn_necklace_C_03}
    componentAppliedLinearAccClamp=<1500 1500 1500>
    componentLinearAccScale=<0.5 0.5 0.5>
    componentLinearVelScale=<0.5 0.5 0.5>
    gravityOverride {z=-980}
    physicsBodyDefinitions [{
      collisionType=InnerSphere
      sphereCollisionRadius=1
      boundBone {boneName=dyn_necklace_C_03}
      boxExtents=<5 5 5>
      constraintSetup {twistAxis=AxisY, angularLimitsMax=<1 5 0>, angularLimitsMin=<-1 -5 -35>, angularTarget=<1 0 0>}
      localJointOffset=<3 0 0>
     }]
    relativeSpaceBone {boneName=root}
   }
   n21 = "Constraint - dyn_necklace_A_01 (PPPPPPPP)"!(
    ComponentPose=@n20
    Alpha=1
    ConstraintWeights_0=0.0092
    ConstraintWeights_1:float=0.0332
    ConstraintWeights_2:float=0.1614
    ConstraintWeights_3:float=0.0232
    ConstraintWeights_4:float=0.0057
    ConstraintWeights_5:float=0.0134
    ConstraintWeights_6:float=0.0031
    ConstraintWeights_7:float=0.7509
   ) -> (Pose) {
    alphaBoolBlend {blendOption=Linear}
    boneToModify {boneName=dyn_necklace_A_01}
    constraintSetup [
     {targetBone {boneName=spine_03}}
     {targetBone {boneName=spine_04}}
     {targetBone {boneName=spine_05}}
     {targetBone {boneName=clavicle_l}}
     {targetBone {boneName=neck_01}}
     {targetBone {boneName=neck_02}}
     {targetBone {boneName=deform_pec_l}}
     {targetBone {boneName=dyn_necklace_A}}
    ]
    constraintWeights [1,1,1,1,1,1,1,1]
   }
   n22 = "Constraint - dyn_necklace_A_02 (PPPPPPPPP)"!(
    ComponentPose=@n21
    Alpha=1
    ConstraintWeights_0=0.0098
    ConstraintWeights_1:float=0.0366
    ConstraintWeights_2:float=0.2038
    ConstraintWeights_3:float=0.0017
    ConstraintWeights_4:float=0.0292
    ConstraintWeights_5:float=0.0084
    ConstraintWeights_6:float=0.0215
    ConstraintWeights_7:float=0.0006
    ConstraintWeights_8:float=0.6885
   ) -> (Pose) {
    alphaBoolBlend {blendOption=Linear}
    boneToModify {boneName=dyn_necklace_A_02}
    constraintSetup [
     {targetBone {boneName=spine_03}}
     {targetBone {boneName=spine_04}}
     {targetBone {boneName=spine_05}}
     {targetBone {boneName=clavicle_l}}
     {targetBone {boneName=clavicle_r}}
     {targetBone {boneName=neck_01}}
     {targetBone {boneName=neck_02}}
     {targetBone {boneName=deform_pec_r}}
     {targetBone {boneName=dyn_necklace_A}}
    ]
    constraintWeights [1,1,1,1,1,1,1,1,1]
   }
   n23 = "Constraint - dyn_necklace_B_01 (PPPPPPPP)"!(
    ComponentPose=@n22
    Alpha=1
    ConstraintWeights_0=0.0086
    ConstraintWeights_1:float=0.0268
    ConstraintWeights_2:float=0.0799
    ConstraintWeights_3:float=0.1088
    ConstraintWeights_4:float=0.0467
    ConstraintWeights_5:float=0.0014
    ConstraintWeights_6:float=0.02
    ConstraintWeights_7:float=0.7078
   ) -> (Pose) {
    alphaBoolBlend {blendOption=Linear}
    boneToModify {boneName=dyn_necklace_B_01}
    constraintSetup [
     {targetBone {boneName=spine_02}}
     {targetBone {boneName=spine_03}}
     {targetBone {boneName=spine_04}}
     {targetBone {boneName=spine_05}}
     {targetBone {boneName=clavicle_l}}
     {targetBone {boneName=neck_02}}
     {targetBone {boneName=deform_pec_l}}
     {targetBone {boneName=dyn_necklace_B}}
    ]
    constraintWeights [1,1,1,1,1,1,1,1]
   }
   n24 = "Constraint - dyn_necklace_B_02 (PPPPPPPP)"!(
    ComponentPose=@n23
    Alpha=1
    ConstraintWeights_0=0.0086
    ConstraintWeights_1:float=0.0269
    ConstraintWeights_2:float=0.0806
    ConstraintWeights_3:float=0.1046
    ConstraintWeights_4:float=0.0428
    ConstraintWeights_5:float=0.0013
    ConstraintWeights_6:float=0.0188
    ConstraintWeights_7:float=0.7164
   ) -> (Pose) {
    alphaBoolBlend {blendOption=Linear}
    boneToModify {boneName=dyn_necklace_B_02}
    constraintSetup [
     {targetBone {boneName=spine_02}}
     {targetBone {boneName=spine_03}}
     {targetBone {boneName=spine_04}}
     {targetBone {boneName=spine_05}}
     {targetBone {boneName=clavicle_r}}
     {targetBone {boneName=neck_02}}
     {targetBone {boneName=deform_pec_r}}
     {targetBone {boneName=dyn_necklace_B}}
    ]
    constraintWeights [1,1,1,1,1,1,1,1]
   }
   n25 = "Constraint - dyn_necklace_C_04 (PPPPPPPP)"!(
    ComponentPose=@n24
    Alpha=1
    ConstraintWeights_0=0.0035
    ConstraintWeights_1:float=0.0123
    ConstraintWeights_2:float=0.0252
    ConstraintWeights_3:float=0.0195
    ConstraintWeights_4:float=0.0065
    ConstraintWeights_5:float=0.0106
    ConstraintWeights_6:float=0.6809
    ConstraintWeights_7:float=0.2416
   ) -> (Pose) {
    alphaBoolBlend {blendOption=Linear}
    boneToModify {boneName=dyn_necklace_C_04}
    constraintSetup [
     {targetBone {boneName=spine_01}}
     {targetBone {boneName=spine_02}}
     {targetBone {boneName=spine_03}}
     {targetBone {boneName=spine_04}}
     {targetBone {boneName=clavicle_l}}
     {targetBone {boneName=deform_pec_l}}
     {targetBone {boneName=dyn_necklace_C_02}}
     {targetBone {boneName=dyn_necklace_C_01}}
    ]
    constraintWeights [1,1,1,1,1,1,1,1]
   }
   n26 = "Constraint - dyn_necklace_C_05 (PPPPPPPPP)"!(
    ComponentPose=@n25
    Alpha=1
    ConstraintWeights_0=0.0005
    ConstraintWeights_1:float=0.0131
    ConstraintWeights_2:float=0.0269
    ConstraintWeights_3:float=0.0206
    ConstraintWeights_4:float=0.0069
    ConstraintWeights_5:float=0.011
    ConstraintWeights_6:float=0.6634
    ConstraintWeights_7:float=0.2523
    ConstraintWeights_8:float=0.0053
   ) -> (Pose) {
    alphaBoolBlend {blendOption=Linear}
    boneToModify {boneName=dyn_necklace_C_05}
    constraintSetup [
     {targetBone {boneName=spine_01}}
     {targetBone {boneName=spine_02}}
     {targetBone {boneName=spine_03}}
     {targetBone {boneName=spine_04}}
     {targetBone {boneName=clavicle_r}}
     {targetBone {boneName=deform_pec_r}}
     {targetBone {boneName=dyn_necklace_C_02}}
     {targetBone {boneName=dyn_necklace_C_01}}
     {targetBone {boneName=dyn_necklace_C}}
    ]
    constraintWeights [1,1,1,1,1,1,1,1,1]
   }
   n27 = Get Arm L rot() -> (Arm L rot:double)
   n28 = "Transform (Modify) Bone - Bone: dyn_necklace_B_02"(ComponentPose=@n26, Alpha=@n27, Translation="0,0,0", Rotation="0,-25,0", Scale="1,1,1") {
    rotationMode=BMM_Additive
    rotationSpace=BCS_BoneSpace
    alphaBoolBlend {blendOption=Linear}
    boneToModify {boneName=dyn_necklace_B_02}
    rotation {yaw=-25}
   }
   n29 = "Transform (Modify) Bone - Bone: dyn_necklace_B_01"(ComponentPose=@n28, Alpha=@n27, Translation="0,0,0", Rotation="0,-15,0", Scale="1,1,1") {
    rotationMode=BMM_Additive
    rotationSpace=BCS_BoneSpace
    alphaBoolBlend {blendOption=Linear}
    boneToModify {boneName=dyn_necklace_B_01}
    rotation {yaw=-15}
   }
   n30 = "Transform (Modify) Bone - Bone: dyn_necklace_A_02"(ComponentPose=@n29, Alpha=@n27, Translation="0,0,0", Rotation="0,-10,0", Scale="1,1,1") {
    rotationMode=BMM_Additive
    rotationSpace=BCS_BoneSpace
    alphaBoolBlend {blendOption=Linear}
    boneToModify {boneName=dyn_necklace_A_02}
    rotation {yaw=-10}
   }
   n31 = "Transform (Modify) Bone - Bone: dyn_necklace_A_01"(ComponentPose=@n30, Alpha=@n27, Translation="0,0,0", Rotation="0,-10,0", Scale="1,1,1") {
    rotationMode=BMM_Additive
    rotationSpace=BCS_BoneSpace
    alphaBoolBlend {blendOption=Linear}
    boneToModify {boneName=dyn_necklace_A_01}
    rotation {yaw=-10}
   }
   n32 = Anim Dynamics - Bone: dyn_necklace_A_01(ComponentPose=@n31, Alpha=1, ExternalForce="0,0,0") {
    angularDampingOverride=0.997
    angularSpringConstant=70
    bAngularSpring=true
    bGravityOverrideInSimSpace=true
    bOverrideAngularDamping=true
    bUseGravityOverride=true
    bUsePlanarLimit=false
    bUseSphericalLimits=true
    simulationSpace=RootRelative
    alphaBoolBlend {blendOption=Linear}
    boundBone {boneName=dyn_necklace_A_01}
    componentAppliedLinearAccClamp=<1500 1500 1500>
    componentLinearAccScale=<0.5 0.5 0.5>
    componentLinearVelScale=<0.5 0.5 0.5>
    gravityOverride {z=-980}
    physicsBodyDefinitions [{
      collisionType=InnerSphere
      sphereCollisionRadius=1
      boundBone {boneName=dyn_necklace_A_01}
      boxExtents=<7 12 12>
      constraintSetup {twistAxis=AxisY, angularLimitsMax=<1 10 0>, angularLimitsMin=<-1 -10 -25>, angularTarget=<1 0 0>}
      localJointOffset=<2 0 0>
     }]
    relativeSpaceBone {boneName=root}
    sphericalLimits [{limitRadius=4, drivingBone {boneName=spine_05}, sphereLocalOffset=<-2 5 -3>}]
   }
   n33 = Anim Dynamics - Bone: dyn_necklace_A_02(ComponentPose=@n32, Alpha=1, ExternalForce="0,0,0") {
    angularDampingOverride=0.997
    angularSpringConstant=70
    bAngularSpring=true
    bGravityOverrideInSimSpace=true
    bOverrideAngularDamping=true
    bUseGravityOverride=true
    bUsePlanarLimit=false
    bUseSphericalLimits=true
    simulationSpace=RootRelative
    alphaBoolBlend {blendOption=Linear}
    boundBone {boneName=dyn_necklace_A_02}
    componentAppliedLinearAccClamp=<1500 1500 1500>
    componentLinearAccScale=<0.5 0.5 0.5>
    componentLinearVelScale=<0.5 0.5 0.5>
    gravityOverride {z=-980}
    physicsBodyDefinitions [{
      collisionType=InnerSphere
      sphereCollisionRadius=1
      boundBone {boneName=dyn_necklace_A_02}
      boxExtents=<7 12 12>
      constraintSetup {twistAxis=AxisY, angularLimitsMax=<1 10 0>, angularLimitsMin=<-1 -10 -25>, angularTarget=<1 0 0>}
      localJointOffset=<2 0 0>
     }]
    relativeSpaceBone {boneName=root}
    sphericalLimits [{limitRadius=4, drivingBone {boneName=spine_05}, sphereLocalOffset=<-2 5 3>}]
   }
   n34 = Anim Dynamics - Bone: dyn_necklace_B_01(ComponentPose=@n33, Alpha=1, ExternalForce="0,0,0") {
    angularDampingOverride=0.997
    angularSpringConstant=70
    bAngularSpring=true
    bGravityOverrideInSimSpace=true
    bOverrideAngularDamping=true
    bUseGravityOverride=true
    bUsePlanarLimit=false
    bUseSphericalLimits=true
    simulationSpace=RootRelative
    alphaBoolBlend {blendOption=Linear}
    boundBone {boneName=dyn_necklace_B_01}
    componentAppliedLinearAccClamp=<1500 1500 1500>
    componentLinearAccScale=<0.5 0.5 0.5>
    componentLinearVelScale=<0.5 0.5 0.5>
    gravityOverride {z=-980}
    physicsBodyDefinitions [{
      collisionType=InnerSphere
      sphereCollisionRadius=1
      boundBone {boneName=dyn_necklace_B_01}
      boxExtents=<7 12 12>
      constraintSetup {twistAxis=AxisY, angularLimitsMax=<1 10 0>, angularLimitsMin=<-1 -10 -25>, angularTarget=<1 0 0>}
      localJointOffset=<2 0 0>
     }]
    relativeSpaceBone {boneName=root}
   }
   n35 = Anim Dynamics - Bone: dyn_necklace_B_02(ComponentPose=@n34, Alpha=1, ExternalForce="0,0,0") {
    angularDampingOverride=0.997
    angularSpringConstant=70
    bAngularSpring=true
    bGravityOverrideInSimSpace=true
    bOverrideAngularDamping=true
    bUseGravityOverride=true
    bUsePlanarLimit=false
    bUseSphericalLimits=true
    simulationSpace=RootRelative
    alphaBoolBlend {blendOption=Linear}
    boundBone {boneName=dyn_necklace_B_02}
    componentAppliedLinearAccClamp=<1500 1500 1500>
    componentLinearAccScale=<0.5 0.5 0.5>
    componentLinearVelScale=<0.5 0.5 0.5>
    gravityOverride {z=-980}
    physicsBodyDefinitions [{
      collisionType=InnerSphere
      sphereCollisionRadius=1
      boundBone {boneName=dyn_necklace_B_02}
      boxExtents=<7 12 12>
      constraintSetup {twistAxis=AxisY, angularLimitsMax=<1 10 0>, angularLimitsMin=<-1 -10 -25>, angularTarget=<1 0 0>}
      localJointOffset=<2 0 0>
     }]
    relativeSpaceBone {boneName=root}
   }
   n36 = Anim Dynamics - Bone: dyn_necklace_C_04(ComponentPose=@n35, Alpha=1, ExternalForce="0,0,0") {
    angularDampingOverride=0.997
    angularSpringConstant=70
    bAngularSpring=true
    bGravityOverrideInSimSpace=true
    bOverrideAngularDamping=true
    bUseGravityOverride=true
    bUsePlanarLimit=false
    bUseSphericalLimits=true
    simulationSpace=RootRelative
    alphaBoolBlend {blendOption=Linear}
    boundBone {boneName=dyn_necklace_C_04}
    componentAppliedLinearAccClamp=<1500 1500 1500>
    componentLinearAccScale=<0.5 0.5 0.5>
    componentLinearVelScale=<0.5 0.5 0.5>
    gravityOverride {z=-980}
    physicsBodyDefinitions [{
      collisionType=InnerSphere
      sphereCollisionRadius=1
      boundBone {boneName=dyn_necklace_C_04}
      boxExtents=<7 12 12>
      constraintSetup {twistAxis=AxisY, angularLimitsMax=<1 10 0>, angularLimitsMin=<-1 -10 -25>, angularTarget=<1 0 0>}
      localJointOffset=<2 0 0>
     }]
    relativeSpaceBone {boneName=root}
   }
   n37 = Anim Dynamics - Bone: dyn_necklace_C_05(ComponentPose=@n36, Alpha=1, ExternalForce="0,0,0") {
    angularDampingOverride=0.997
    angularSpringConstant=70
    bAngularSpring=true
    bGravityOverrideInSimSpace=true
    bOverrideAngularDamping=true
    bUseGravityOverride=true
    bUsePlanarLimit=false
    bUseSphericalLimits=true
    simulationSpace=RootRelative
    alphaBoolBlend {blendOption=Linear}
    boundBone {boneName=dyn_necklace_C_05}
    componentAppliedLinearAccClamp=<1500 1500 1500>
    componentLinearAccScale=<0.5 0.5 0.5>
    componentLinearVelScale=<0.5 0.5 0.5>
    gravityOverride {z=-980}
    physicsBodyDefinitions [{
      collisionType=InnerSphere
      sphereCollisionRadius=1
      boundBone {boneName=dyn_necklace_C_05}
      boxExtents=<7 12 12>
      constraintSetup {twistAxis=AxisY, angularLimitsMax=<1 10 0>, angularLimitsMin=<-1 -10 -25>, angularTarget=<1 0 0>}
      localJointOffset=<2 0 0>
     }]
    relativeSpaceBone {boneName=root}
   }
   n38 = "Transform (Modify) Bone - Bone: dyn_bk_cloth_MSTR"(ComponentPose=@n37, Alpha=@n6, Translation="0,0,0", Rotation="0,-10,0", Scale="1,1,1") {
    rotationMode=BMM_Additive
    rotationSpace=BCS_BoneSpace
    translationMode=BMM_Additive
    alphaBoolBlend {blendOption=Linear}
    alphaScaleBiasClamp {bInterpResult=true}
    boneToModify {boneName=dyn_bk_cloth_MSTR}
    rotation {yaw=-10}
   }
   n39 = "Transform (Modify) Bone - Bone: dyn_r_cloth_MSTR"(ComponentPose=@n38, Alpha=@n6, Translation="0,0,0", Rotation="0,-7,0", Scale="1,1,1") {
    rotationMode=BMM_Additive
    rotationSpace=BCS_BoneSpace
    translationMode=BMM_Additive
    alphaBoolBlend {blendOption=Linear}
    alphaScaleBiasClamp {bInterpResult=true}
    boneToModify {boneName=dyn_r_cloth_MSTR}
    rotation {yaw=-7}
   }
   n40 = "Transform (Modify) Bone - Bone: dyn_l_cloth_MSTR"(ComponentPose=@n39, Alpha=@n6, Translation="0,0,0", Rotation="0,-7,0", Scale="1,1,1") {
    rotationMode=BMM_Additive
    rotationSpace=BCS_BoneSpace
    translationMode=BMM_Additive
    alphaBoolBlend {blendOption=Linear}
    alphaScaleBiasClamp {bInterpResult=true}
    boneToModify {boneName=dyn_l_cloth_MSTR}
    rotation {yaw=-7}
   }
   n41 = Get backbling_cloth() -> (backbling_cloth:double)
   n42 = "Constraint - dyn_l_cloth_01 (R)"(ComponentPose=@n40, Alpha=@n41, ConstraintWeights_0=1) {
    alphaBoolBlend {blendOption=Linear}
    boneToModify {boneName=dyn_l_cloth_01}
    constraintSetup [{transformType=Rotation, targetBone {boneName=thigh_l}}]
    constraintWeights [1]
   }
   n43 = "Constraint - dyn_l_cloth_03 (R)"(ComponentPose=@n42, Alpha=@n41, ConstraintWeights_0=1) {
    alphaBoolBlend {blendOption=Linear}
    boneToModify {boneName=dyn_l_cloth_03}
    constraintSetup [{transformType=Rotation, targetBone {boneName=calf_l}}]
    constraintWeights [1]
   }
   n44 = "Constraint - dyn_r_cloth_01 (R)"(ComponentPose=@n43, Alpha=@n41, ConstraintWeights_0=1) {
    alphaBoolBlend {blendOption=Linear}
    boneToModify {boneName=dyn_r_cloth_01}
    constraintSetup [{transformType=Rotation, targetBone {boneName=thigh_r}}]
    constraintWeights [1]
   }
   n45 = "Constraint - dyn_r_cloth_03 (R)"(ComponentPose=@n44, Alpha=@n41, ConstraintWeights_0=1) {
    alphaBoolBlend {blendOption=Linear}
    boneToModify {boneName=dyn_r_cloth_03}
    constraintSetup [{transformType=Rotation, targetBone {boneName=calf_r}}]
    constraintWeights [1]
   }
   n46 = "Constraint - dyn_bk_cloth_01 (RR)"(ComponentPose=@n45, Alpha=@n41, ConstraintWeights_0=0.5, ConstraintWeights_1=0.5) {
    alphaBoolBlend {blendOption=Linear}
    boneToModify {boneName=dyn_bk_cloth_01}
    constraintSetup [{transformType=Rotation, targetBone {boneName=dyn_l_cloth_01}}, {transformType=Rotation, targetBone {boneName=dyn_r_cloth_01}}]
    constraintWeights [1,1]
   }
   n47 = "Constraint - dyn_bk_cloth_03 (RR)"(ComponentPose=@n46, Alpha=@n41, ConstraintWeights_0=0.5, ConstraintWeights_1=0.5) {
    alphaBoolBlend {blendOption=Linear}
    boneToModify {boneName=dyn_bk_cloth_03}
    constraintSetup [{transformType=Rotation, targetBone {boneName=dyn_l_cloth_03}}, {transformType=Rotation, targetBone {boneName=dyn_r_cloth_03}}]
    constraintWeights [1,1]
   }
   n48 = Anim Dynamics - Bone: dyn_skirt_fr_l_01(ComponentPose=@n47, Alpha=1, ExternalForce="0,0,0") {
    angularDampingOverride=0.9999
    angularSpringConstant=150
    bAngularSpring=true
    bGravityOverrideInSimSpace=true
    bOverrideAngularDamping=true
    bUsePlanarLimit=false
    simulationSpace=RootRelative
    alphaBoolBlend {blendOption=Linear}
    boundBone {boneName=dyn_skirt_fr_l_01}
    componentAppliedLinearAccClamp=<3500 3500 3500>
    componentLinearAccScale=<1 1 1>
    componentLinearVelScale=<1 1 1>
    physicsBodyDefinitions [{
      sphereCollisionRadius=0
      boundBone {boneName=dyn_skirt_fr_l_01}
      boxExtents=<25 10 10>
      constraintSetup {angularLimitsMax=<1 5 0>, angularLimitsMin=<-1 -5 -35>, angularTarget=<0 0 0>}
      localJointOffset=<2 0 0>
     }]
   }
   n49 = Anim Dynamics - Bone: dyn_skirt_fr_l_02(ComponentPose=@n48, Alpha=1, ExternalForce="0,0,0") {
    angularDampingOverride=0.9999
    angularSpringConstant=150
    bAngularSpring=true
    bGravityOverrideInSimSpace=true
    bOverrideAngularDamping=true
    bUsePlanarLimit=false
    simulationSpace=RootRelative
    alphaBoolBlend {blendOption=Linear}
    boundBone {boneName=dyn_skirt_fr_l_02}
    componentAppliedLinearAccClamp=<3500 3500 3500>
    componentLinearAccScale=<1 1 1>
    componentLinearVelScale=<1 1 1>
    physicsBodyDefinitions [{
      sphereCollisionRadius=0
      boundBone {boneName=dyn_skirt_fr_l_02}
      boxExtents=<25 10 10>
      constraintSetup {angularLimitsMax=<1 5 0>, angularLimitsMin=<-1 -5 -35>, angularTarget=<0 0 0>}
      localJointOffset=<2 0 0>
     }]
   }
   n50 = Anim Dynamics - Bone: dyn_skirt_fr_l_02(ComponentPose=@n49, Alpha=1, ExternalForce="0,0,0") {
    angularDampingOverride=0.9999
    angularSpringConstant=150
    bAngularSpring=true
    bGravityOverrideInSimSpace=true
    bOverrideAngularDamping=true
    bUsePlanarLimit=false
    simulationSpace=RootRelative
    alphaBoolBlend {blendOption=Linear}
    boundBone {boneName=dyn_skirt_fr_l_02}
    componentAppliedLinearAccClamp=<3500 3500 3500>
    componentLinearAccScale=<1 1 1>
    componentLinearVelScale=<1 1 1>
    physicsBodyDefinitions [{
      sphereCollisionRadius=0
      boundBone {boneName=dyn_skirt_fr_l_02}
      boxExtents=<25 10 10>
      constraintSetup {angularLimitsMax=<1 5 0>, angularLimitsMin=<-1 -5 -35>, angularTarget=<0 0 0>}
      localJointOffset=<2 0 0>
     }]
   }
   n51 = Anim Dynamics - Bone: dyn_skirt_fr_r_01(ComponentPose=@n50, Alpha=1, ExternalForce="0,0,0") {
    angularDampingOverride=0.9999
    angularSpringConstant=150
    bAngularSpring=true
    bGravityOverrideInSimSpace=true
    bOverrideAngularDamping=true
    bUsePlanarLimit=false
    simulationSpace=RootRelative
    alphaBoolBlend {blendOption=Linear}
    boundBone {boneName=dyn_skirt_fr_r_01}
    componentAppliedLinearAccClamp=<3500 3500 3500>
    componentLinearAccScale=<1 1 1>
    componentLinearVelScale=<1 1 1>
    physicsBodyDefinitions [{
      sphereCollisionRadius=0
      boundBone {boneName=dyn_skirt_fr_r_01}
      boxExtents=<25 10 10>
      constraintSetup {angularLimitsMax=<1 5 0>, angularLimitsMin=<-1 -5 -35>, angularTarget=<0 0 0>}
      localJointOffset=<2 0 0>
     }]
   }
   n52 = Anim Dynamics - Bone: dyn_skirt_fr_r_02(ComponentPose=@n51, Alpha=1, ExternalForce="0,0,0") {
    angularDampingOverride=0.9999
    angularSpringConstant=150
    bAngularSpring=true
    bGravityOverrideInSimSpace=true
    bOverrideAngularDamping=true
    bUsePlanarLimit=false
    simulationSpace=RootRelative
    alphaBoolBlend {blendOption=Linear}
    boundBone {boneName=dyn_skirt_fr_r_02}
    componentAppliedLinearAccClamp=<3500 3500 3500>
    componentLinearAccScale=<1 1 1>
    componentLinearVelScale=<1 1 1>
    physicsBodyDefinitions [{
      sphereCollisionRadius=0
      boundBone {boneName=dyn_skirt_fr_r_02}
      boxExtents=<25 10 10>
      constraintSetup {angularLimitsMax=<1 5 0>, angularLimitsMin=<-1 -5 -35>, angularTarget=<0 0 0>}
      localJointOffset=<2 0 0>
     }]
   }
   n53 = Anim Dynamics - Bone: dyn_skirt_fr_r_02(ComponentPose=@n52, Alpha=1, ExternalForce="0,0,0") {
    angularDampingOverride=0.9999
    angularSpringConstant=150
    bAngularSpring=true
    bGravityOverrideInSimSpace=true
    bOverrideAngularDamping=true
    bUsePlanarLimit=false
    simulationSpace=RootRelative
    alphaBoolBlend {blendOption=Linear}
    boundBone {boneName=dyn_skirt_fr_r_02}
    componentAppliedLinearAccClamp=<3500 3500 3500>
    componentLinearAccScale=<1 1 1>
    componentLinearVelScale=<1 1 1>
    physicsBodyDefinitions [{
      sphereCollisionRadius=0
      boundBone {boneName=dyn_skirt_fr_r_02}
      boxExtents=<25 10 10>
      constraintSetup {angularLimitsMax=<1 5 0>, angularLimitsMin=<-1 -5 -35>, angularTarget=<0 0 0>}
      localJointOffset=<2 0 0>
     }]
   }
   n54 = "Constraint - dyn_skirt_fr_r_02 (P)"(ComponentPose=@n53, Alpha=Get crouch pose() -> (crouch pose:double), ConstraintWeights_0=1) {
    alphaBoolBlend {blendOption=Linear}
    alphaScaleBiasClamp {bInterpResult=true}
    boneToModify {boneName=dyn_skirt_fr_r_02}
    constraintSetup [{targetBone {boneName=calf_r}}]
    constraintWeights [1]
   }
   n55 = Get bIsSkydiveFloating()
   n56 = "Constraint - dyn_skirt_fr_r_02 (P)"!(ComponentPose=@n54, bAlphaBoolEnabled:bool=@n55, ConstraintWeights_0=1) -> (Pose) {
    alphaInputType=Bool
    alphaBoolBlend {blendInTime=0.1, blendOption=Linear, blendOutTime=1}
    alphaScaleBiasClamp {bInterpResult=true}
    boneToModify {boneName=dyn_skirt_fr_r_02}
    constraintSetup [{targetBone {boneName=calf_r}}]
    constraintWeights [1]
   }
   n57 = "Constraint - dyn_skirt_fr_l_02 (P)"!(ComponentPose=@n56, bAlphaBoolEnabled:bool=@n55, ConstraintWeights_0=1) -> (Pose) {
    alphaInputType=Bool
    alphaBoolBlend {blendInTime=0.1, blendOption=Linear, blendOutTime=1}
    alphaScaleBiasClamp {bInterpResult=true}
    boneToModify {boneName=dyn_skirt_fr_l_02}
    constraintSetup [{targetBone {boneName=calf_l}}]
    constraintWeights [1]
   }
   n58 = Get bIsSkydiveDiveMode()
   n59 = "Constraint - dyn_skirt_fr_r_02 (P)"(ComponentPose=@n57, bAlphaBoolEnabled=@n58, ConstraintWeights_0=1) {
    alphaInputType=Bool
    alphaBoolBlend {blendInTime=0.1, blendOption=Linear, blendOutTime=1}
    alphaScaleBiasClamp {bInterpResult=true}
    boneToModify {boneName=dyn_skirt_fr_r_02}
    constraintSetup [{targetBone {boneName=calf_r}}]
    constraintWeights [1]
   }
   n60 = "Constraint - dyn_skirt_fr_l_02 (P)"(ComponentPose=@n59, bAlphaBoolEnabled=@n58, ConstraintWeights_0=1) {
    alphaInputType=Bool
    alphaBoolBlend {blendInTime=0.1, blendOption=Linear, blendOutTime=1}
    alphaScaleBiasClamp {bInterpResult=true}
    boneToModify {boneName=dyn_skirt_fr_l_02}
    constraintSetup [{targetBone {boneName=calf_l}}]
    constraintWeights [1]
   }
   n61 = Anim Dynamics - Bone: dyn_wristband_01(ComponentPose=@n60, Alpha=1, ExternalForce="0,0,0") {
    angularDampingOverride=0.997
    angularSpringConstant=70
    bAngularSpring=true
    bGravityOverrideInSimSpace=true
    bOverrideAngularDamping=true
    bUseGravityOverride=true
    bUsePlanarLimit=false
    bUseSphericalLimits=true
    simulationSpace=RootRelative
    alphaBoolBlend {blendOption=Linear}
    boundBone {boneName=dyn_wristband_01}
    componentAppliedLinearAccClamp=<1500 1500 1500>
    componentLinearAccScale=<0.5 0.5 0.5>
    componentLinearVelScale=<0.5 0.5 0.5>
    gravityOverride {z=-980}
    physicsBodyDefinitions [{
      collisionType=InnerSphere
      sphereCollisionRadius=1
      boundBone {boneName=dyn_wristband_01}
      boxExtents=<7 12 12>
      constraintSetup {twistAxis=AxisY, angularLimitsMax=<1 0 5>, angularLimitsMin=<-1 -15 -5>, angularTarget=<1 0 0>}
      localJointOffset=<2 0 0>
     }]
    relativeSpaceBone {boneName=root}
   }
   n62 = Anim Dynamics - Bone: dyn_wristband_02(ComponentPose=@n61, Alpha=1, ExternalForce="0,0,0") {
    angularDampingOverride=0.997
    angularSpringConstant=70
    bAngularSpring=true
    bGravityOverrideInSimSpace=true
    bOverrideAngularDamping=true
    bUseGravityOverride=true
    bUsePlanarLimit=false
    bUseSphericalLimits=true
    simulationSpace=RootRelative
    alphaBoolBlend {blendOption=Linear}
    boundBone {boneName=dyn_wristband_02}
    componentAppliedLinearAccClamp=<1500 1500 1500>
    componentLinearAccScale=<0.5 0.5 0.5>
    componentLinearVelScale=<0.5 0.5 0.5>
    gravityOverride {z=-980}
    physicsBodyDefinitions [{
      collisionType=InnerSphere
      sphereCollisionRadius=1
      boundBone {boneName=dyn_wristband_02}
      boxExtents=<7 12 12>
      constraintSetup {twistAxis=AxisY, angularLimitsMax=<1 0 5>, angularLimitsMin=<-1 -15 -5>, angularTarget=<1 0 0>}
      localJointOffset=<2 0 0>
     }]
    relativeSpaceBone {boneName=root}
   }
   n63 = Anim Dynamics - Bone: dyn_wristband_03(ComponentPose=@n62, Alpha=1, ExternalForce="0,0,0") {
    angularDampingOverride=0.997
    angularSpringConstant=70
    bAngularSpring=true
    bGravityOverrideInSimSpace=true
    bOverrideAngularDamping=true
    bUseGravityOverride=true
    bUsePlanarLimit=false
    bUseSphericalLimits=true
    simulationSpace=RootRelative
    alphaBoolBlend {blendOption=Linear}
    boundBone {boneName=dyn_wristband_03}
    componentAppliedLinearAccClamp=<1500 1500 1500>
    componentLinearAccScale=<0.5 0.5 0.5>
    componentLinearVelScale=<0.5 0.5 0.5>
    gravityOverride {z=-980}
    physicsBodyDefinitions [{
      collisionType=InnerSphere
      sphereCollisionRadius=1
      boundBone {boneName=dyn_wristband_03}
      boxExtents=<7 12 12>
      constraintSetup {twistAxis=AxisY, angularLimitsMax=<1 0 5>, angularLimitsMin=<-1 -15 -5>, angularTarget=<1 0 0>}
      localJointOffset=<2 0 0>
     }]
    relativeSpaceBone {boneName=root}
   }
   n64 = Anim Dynamics - Bone: dyn_wristband_04(ComponentPose=@n63, Alpha=1, ExternalForce="0,0,0") {
    angularDampingOverride=0.997
    angularSpringConstant=70
    bAngularSpring=true
    bGravityOverrideInSimSpace=true
    bOverrideAngularDamping=true
    bUseGravityOverride=true
    bUsePlanarLimit=false
    bUseSphericalLimits=true
    simulationSpace=RootRelative
    alphaBoolBlend {blendOption=Linear}
    boundBone {boneName=dyn_wristband_04}
    componentAppliedLinearAccClamp=<1500 1500 1500>
    componentLinearAccScale=<0.5 0.5 0.5>
    componentLinearVelScale=<0.5 0.5 0.5>
    gravityOverride {z=-980}
    physicsBodyDefinitions [{
      collisionType=InnerSphere
      sphereCollisionRadius=1
      boundBone {boneName=dyn_wristband_04}
      boxExtents=<7 12 12>
      constraintSetup {twistAxis=AxisY, angularLimitsMax=<1 0 5>, angularLimitsMin=<-1 -15 -5>, angularTarget=<1 0 0>}
      localJointOffset=<2 0 0>
     }]
    relativeSpaceBone {boneName=root}
   }
   n65 = "Transform (Modify) Bone - Bone: dyn_necklace_C_02"(ComponentPose=@n64, Alpha=Get ThighRotL() -> (ThighRotL:double), Translation="0,0,2", Rotation="25,-15,15", Scale="1,1,1") {
    rotationMode=BMM_Additive
    rotationSpace=BCS_BoneSpace
    translationMode=BMM_Additive
    translationSpace=BCS_BoneSpace
    alphaBoolBlend {blendOption=Linear}
    alphaScaleBiasClamp {bInterpResult=true}
    boneToModify {boneName=dyn_necklace_C_02}
    rotation=<rot 25 -15 15>
    translation {z=2}
   }
   Output Pose(Result=Component To Local(ComponentPose=@n65)) {layerGroup=None}
   n66 = "Constraint - dyn_jacket_fr_l_MSTR (PPPPPP)"!(
    ComponentPose=Local To Component(LocalPose=Copy Pose From Mesh() {bUseAttachedParent=true})
    Alpha=1
    ConstraintWeights_0=0.7377
    ConstraintWeights_1:float=0.0895
    ConstraintWeights_2:float=0.1044
    ConstraintWeights_3:float=0.0588
    ConstraintWeights_4:float=0.0059
    ConstraintWeights_5:float=0.0038
   ) -> (Pose) {
    alphaBoolBlend {blendOption=Linear}
    boneToModify {boneName=dyn_jacket_fr_l_MSTR}
    constraintSetup [
     {targetBone {boneName=pelvis}}
     {targetBone {boneName=spine_01}}
     {targetBone {boneName=spine_02}}
     {targetBone {boneName=spine_03}}
     {targetBone {boneName=thigh_twist_01_l}}
     {targetBone {boneName=deform_groin_l}}
    ]
    constraintWeights [1,1,1,1,1,1]
   }
   n67 = "Constraint - dyn_jacket_l_MSTR (PPPPP)"!(
    ComponentPose=@n66
    Alpha=1
    ConstraintWeights_0=0.7547
    ConstraintWeights_1:float=0.0862
    ConstraintWeights_2:float=0.1011
    ConstraintWeights_3:float=0.0572
    ConstraintWeights_4:float=0.0008
   ) -> (Pose) {
    alphaBoolBlend {blendOption=Linear}
    boneToModify {boneName=dyn_jacket_l_MSTR}
    constraintSetup [
     {targetBone {boneName=pelvis}}
     {targetBone {boneName=spine_01}}
     {targetBone {boneName=spine_02}}
     {targetBone {boneName=spine_03}}
     {targetBone {boneName=thigh_twist_01_l}}
    ]
    constraintWeights [1,1,1,1,1]
   }
   n68 = "Constraint - dyn_jacket_bk_MSTR (PPPPPPPPP)"!(
    ComponentPose=@n67
    Alpha=1
    ConstraintWeights_0=0.7041
    ConstraintWeights_1:float=0.1059
    ConstraintWeights_2:float=0.1062
    ConstraintWeights_3:float=0.0757
    ConstraintWeights_4:float=0.0077
    ConstraintWeights_5:float=0
    ConstraintWeights_6:float=0.0002
    ConstraintWeights_7:float=0
    ConstraintWeights_8:float=0.0002
   ) -> (Pose) {
    alphaBoolBlend {blendOption=Linear}
    boneToModify {boneName=dyn_jacket_bk_MSTR}
    constraintSetup [
     {targetBone {boneName=pelvis}}
     {targetBone {boneName=spine_01}}
     {targetBone {boneName=spine_02}}
     {targetBone {boneName=spine_03}}
     {targetBone {boneName=spine_04}}
     {targetBone {boneName=thigh_twist_01_l}}
     {targetBone {boneName=deform_glute_l}}
     {targetBone {boneName=thigh_twist_01_r}}
     {targetBone {boneName=deform_glute_r}}
    ]
    constraintWeights [1,1,1,1,1,1,1,1,1]
   }
   n69 = "Constraint - dyn_jacket_r_MSTR (PPPPP)"!(
    ComponentPose=@n68
    Alpha=1
    ConstraintWeights_0=0.7547
    ConstraintWeights_1:float=0.0862
    ConstraintWeights_2:float=0.1011
    ConstraintWeights_3:float=0.0572
    ConstraintWeights_4:float=0.0008
   ) -> (Pose) {
    alphaBoolBlend {blendOption=Linear}
    boneToModify {boneName=dyn_jacket_r_MSTR}
    constraintSetup [
     {targetBone {boneName=pelvis}}
     {targetBone {boneName=spine_01}}
     {targetBone {boneName=spine_02}}
     {targetBone {boneName=spine_03}}
     {targetBone {boneName=thigh_twist_01_r}}
    ]
    constraintWeights [1,1,1,1,1]
   }
   n70 = "Constraint - dyn_jacket_fr_r_MSTR (PPPPPP)"!(
    ComponentPose=@n69
    Alpha=1
    ConstraintWeights_0=0.7385
    ConstraintWeights_1:float=0.0891
    ConstraintWeights_2:float=0.104
    ConstraintWeights_3:float=0.0586
    ConstraintWeights_4:float=0.006
    ConstraintWeights_5:float=0.0038
   ) -> (Pose) {
    alphaBoolBlend {blendOption=Linear}
    boneToModify {boneName=dyn_jacket_fr_r_MSTR}
    constraintSetup [
     {targetBone {boneName=pelvis}}
     {targetBone {boneName=spine_01}}
     {targetBone {boneName=spine_02}}
     {targetBone {boneName=spine_03}}
     {targetBone {boneName=thigh_twist_01_r}}
     {targetBone {boneName=deform_groin_r}}
    ]
    constraintWeights [1,1,1,1,1,1]
   }
   n71 = "Constraint - dyn_bk_cloth_MSTR (PPPPP)"!(
    ComponentPose=@n70
    Alpha=1
    ConstraintWeights_0=0.6001
    ConstraintWeights_1:float=0.1161
    ConstraintWeights_2:float=0.2348
    ConstraintWeights_3:float=0.0001
    ConstraintWeights_4:float=0.0489
   ) -> (Pose) {
    alphaBoolBlend {blendOption=Linear}
    boneToModify {boneName=dyn_bk_cloth_MSTR}
    constraintSetup [
     {targetBone {boneName=pelvis}}
     {targetBone {boneName=thigh_twist_01_l}}
     {targetBone {boneName=deform_glute_l}}
     {targetBone {boneName=deform_groin_l}}
     {targetBone {boneName=thigh_twist_01_r}}
    ]
    constraintWeights [1,1,1,1,1]
   }
   n72 = "Constraint - dyn_l_cloth_MSTR (PPPP)"!(
    ComponentPose=@n71
    Alpha=1
    ConstraintWeights_0=0.1295
    ConstraintWeights_1:float=0.0203
    ConstraintWeights_2:float=0.6347
    ConstraintWeights_3:float=0.2155
   ) -> (Pose) {
    alphaBoolBlend {blendOption=Linear}
    boneToModify {boneName=dyn_l_cloth_MSTR}
    constraintSetup [
     {targetBone {boneName=pelvis}}
     {targetBone {boneName=thigh_l}}
     {targetBone {boneName=thigh_twist_01_l}}
     {targetBone {boneName=deform_glute_l}}
    ]
    constraintWeights [1,1,1,1]
   }
   n73 = "Constraint - dyn_r_cloth_MSTR (PPPPP)"!(
    ComponentPose=@n72
    Alpha=1
    ConstraintWeights_0=0.1302
    ConstraintWeights_1:float=0.0072
    ConstraintWeights_2:float=0.0201
    ConstraintWeights_3:float=0.6287
    ConstraintWeights_4:float=0.2137
   ) -> (Pose) {
    alphaBoolBlend {blendOption=Linear}
    boneToModify {boneName=dyn_r_cloth_MSTR}
    constraintSetup [
     {targetBone {boneName=pelvis}}
     {targetBone {boneName=deform_glute_l}}
     {targetBone {boneName=thigh_r}}
     {targetBone {boneName=thigh_twist_01_r}}
     {targetBone {boneName=deform_glute_r}}
    ]
    constraintWeights [1,1,1,1,1]
   }
   n74 = "Constraint - dyn_skirt_fr_l_MSTR (PPPPPP)"!(
    ComponentPose=@n73
    Alpha=1
    ConstraintWeights_0=0.4276
    ConstraintWeights_1:float=0.1266
    ConstraintWeights_2:float=0.0052
    ConstraintWeights_3:float=0.3883
    ConstraintWeights_4:float=0.0021
    ConstraintWeights_5:float=0.0501
   ) -> (Pose) {
    alphaBoolBlend {blendOption=Linear}
    boneToModify {boneName=dyn_skirt_fr_l_MSTR}
    constraintSetup [
     {targetBone {boneName=thigh_l}}
     {targetBone {boneName=calf_l}}
     {targetBone {boneName=calf_twist_02_l}}
     {targetBone {boneName=deform_knee_out_l}}
     {targetBone {boneName=deform_knee_in_l}}
     {targetBone {boneName=thigh_twist_01_l}}
    ]
    constraintWeights [1,1,1,1,1,1]
   }
   n75 = "Constraint - dyn_skirt_fr_r_MSTR (PPPPPP)"!(
    ComponentPose=@n74
    Alpha=1
    ConstraintWeights_0=0.4276
    ConstraintWeights_1:float=0.1266
    ConstraintWeights_2:float=0.0052
    ConstraintWeights_3:float=0.3883
    ConstraintWeights_4:float=0.0021
    ConstraintWeights_5:float=0.0501
   ) -> (Pose) {
    alphaBoolBlend {blendOption=Linear}
    boneToModify {boneName=dyn_skirt_fr_r_MSTR}
    constraintSetup [
     {targetBone {boneName=thigh_r}}
     {targetBone {boneName=calf_r}}
     {targetBone {boneName=calf_twist_02_r}}
     {targetBone {boneName=deform_knee_out_r}}
     {targetBone {boneName=deform_knee_in_r}}
     {targetBone {boneName=thigh_twist_01_r}}
    ]
    constraintWeights [1,1,1,1,1,1]
   }
   n76 = "Constraint - dyn_necklace_A (PPPPPP)"!(
    ComponentPose=@n75
    Alpha=1
    ConstraintWeights_0=0.6654
    ConstraintWeights_1:float=0.064
    ConstraintWeights_2:float=0.0641
    ConstraintWeights_3:float=0.0557
    ConstraintWeights_4:float=0.1386
    ConstraintWeights_5:float=0.0121
   ) -> (Pose) {
    alphaBoolBlend {blendOption=Linear}
    boneToModify {boneName=dyn_necklace_A}
    constraintSetup [
     {targetBone {boneName=spine_05}}
     {targetBone {boneName=clavicle_l}}
     {targetBone {boneName=clavicle_r}}
     {targetBone {boneName=neck_01}}
     {targetBone {boneName=neck_02}}
     {targetBone {boneName=head}}
    ]
    constraintWeights [1,1,1,1,1,1]
   }
   n77 = "Constraint - dyn_necklace_B (PPPPPPPPPPPPP)"!(
    ComponentPose=@n76
    Alpha=1
    ConstraintWeights_0=0.0158
    ConstraintWeights_1:float=0.036
    ConstraintWeights_2:float=0.4947
    ConstraintWeights_3:float=0.1988
    ConstraintWeights_4:float=0.0012
    ConstraintWeights_5:float=0.0018
    ConstraintWeights_6:float=0.1788
    ConstraintWeights_7:float=0.0015
    ConstraintWeights_8:float=0.0016
    ConstraintWeights_9:float=0.0241
    ConstraintWeights_10:float=0.0098
    ConstraintWeights_11:float=0.0059
    ConstraintWeights_12:float=0.0299
   ) -> (Pose) {
    alphaBoolBlend {blendOption=Linear}
    boneToModify {boneName=dyn_necklace_B}
    constraintSetup [
     {targetBone {boneName=spine_03}}
     {targetBone {boneName=spine_04}}
     {targetBone {boneName=spine_05}}
     {targetBone {boneName=clavicle_l}}
     {targetBone {boneName=upperarm_twist_01_l}}
     {targetBone {boneName=deform_upperarm_up_l}}
     {targetBone {boneName=clavicle_r}}
     {targetBone {boneName=deform_upperarm_up_r}}
     {targetBone {boneName=neck_01}}
     {targetBone {boneName=neck_02}}
     {targetBone {boneName=deform_pec_r}}
     {targetBone {boneName=deform_pec_l}}
     {targetBone {boneName=dyn_necklace_B}}
    ]
    constraintWeights [1,1,1,1,1,1,1,1,1,1,1,1,1]
   }
   n78 = "Constraint - dyn_necklace_C (PPPPPPPP)"!(
    ComponentPose=@n77
    Alpha=1
    ConstraintWeights_0=0.0076
    ConstraintWeights_1:float=0.0446
    ConstraintWeights_2:float=0.1692
    ConstraintWeights_3:float=0.6581
    ConstraintWeights_4:float=0.0012
    ConstraintWeights_5:float=0.0503
    ConstraintWeights_6:float=0.0208
    ConstraintWeights_7:float=0.048
   ) -> (Pose) {
    alphaBoolBlend {blendOption=Linear}
    boneToModify {boneName=dyn_necklace_C}
    constraintSetup [
     {targetBone {boneName=spine_02}}
     {targetBone {boneName=spine_03}}
     {targetBone {boneName=spine_04}}
     {targetBone {boneName=spine_05}}
     {targetBone {boneName=clavicle_l}}
     {targetBone {boneName=clavicle_r}}
     {targetBone {boneName=neck_01}}
     {targetBone {boneName=neck_02}}
    ]
    constraintWeights [1,1,1,1,1,1,1,1]
   }
   n79 = "Constraint - dyn_necklace_C_01 (PPPPPPPP)"!(
    ComponentPose=@n78
    Alpha=1
    ConstraintWeights_0=0.082
    ConstraintWeights_1:float=0.1922
    ConstraintWeights_2:float=0.4355
    ConstraintWeights_3:float=0.2254
    ConstraintWeights_4:float=0.0143
    ConstraintWeights_5:float=0.0146
    ConstraintWeights_6:float=0.018
    ConstraintWeights_7:float=0.018
   ) -> (Pose) {
    alphaBoolBlend {blendOption=Linear}
    boneToModify {boneName=dyn_necklace_C_01}
    constraintSetup [
     {targetBone {boneName=spine_02}}
     {targetBone {boneName=spine_03}}
     {targetBone {boneName=spine_04}}
     {targetBone {boneName=spine_05}}
     {targetBone {boneName=clavicle_l}}
     {targetBone {boneName=clavicle_r}}
     {targetBone {boneName=deform_pec_r}}
     {targetBone {boneName=deform_pec_l}}
    ]
    constraintWeights [1,1,1,1,1,1,1,1]
   }
   n80 = "Constraint - dyn_necklace_C_02 (PPPPPPPP)"!(
    ComponentPose=@n79
    Alpha=1
    ConstraintWeights_0=0.0432
    ConstraintWeights_1:float=0.1761
    ConstraintWeights_2:float=0.3459
    ConstraintWeights_3:float=0.3341
    ConstraintWeights_4:float=0.07
    ConstraintWeights_5:float=0.0101
    ConstraintWeights_6:float=0.0103
    ConstraintWeights_7:float=0.0104
   ) -> (Pose) {
    alphaBoolBlend {blendOption=Linear}
    boneToModify {boneName=dyn_necklace_C_02}
    constraintSetup [
     {targetBone {boneName=spine_01}}
     {targetBone {boneName=spine_02}}
     {targetBone {boneName=spine_03}}
     {targetBone {boneName=spine_04}}
     {targetBone {boneName=spine_05}}
     {targetBone {boneName=clavicle_r}}
     {targetBone {boneName=deform_pec_r}}
     {targetBone {boneName=deform_pec_l}}
    ]
    constraintWeights [1,1,1,1,1,1,1,1]
   }
   n81 = "Constraint - dyn_wristband_01 (PPPPPPPP)"!(
    ComponentPose=@n80
    Alpha=1
    ConstraintWeights_0=0.1912
    ConstraintWeights_1:float=0.0059
    ConstraintWeights_2:float=0.0044
    ConstraintWeights_3:float=0.065
    ConstraintWeights_4:float=0.0138
    ConstraintWeights_5:float=0.3101
    ConstraintWeights_6:float=0.3682
    ConstraintWeights_7:float=0.0413
   ) -> (Pose) {
    alphaBoolBlend {blendOption=Linear}
    boneToModify {boneName=dyn_wristband_01}
    constraintSetup [
     {targetBone {boneName=hand_r}}
     {targetBone {boneName=index_metacarpal_r}}
     {targetBone {boneName=middle_metacarpal_r}}
     {targetBone {boneName=pinky_metacarpal_r}}
     {targetBone {boneName=ring_metacarpal_r}}
     {targetBone {boneName=deform_wrist_up_r}}
     {targetBone {boneName=lowerarm_twist_01_r}}
     {targetBone {boneName=lowerarm_twist_02_r}}
    ]
    constraintWeights [1,1,1,1,1,1,1,1]
   }
   n82 = "Constraint - dyn_wristband_02 (PPPPPPPP)"!(
    ComponentPose=@n81
    Alpha=1
    ConstraintWeights_0=0.161
    ConstraintWeights_1:float=0.038
    ConstraintWeights_2:float=0.0066
    ConstraintWeights_3:float=0.0078
    ConstraintWeights_4:float=0.0663
    ConstraintWeights_5:float=0.2519
    ConstraintWeights_6:float=0.4137
    ConstraintWeights_7:float=0.0547
   ) -> (Pose) {
    alphaBoolBlend {blendOption=Linear}
    boneToModify {boneName=dyn_wristband_02}
    constraintSetup [
     {targetBone {boneName=hand_r}}
     {targetBone {boneName=index_metacarpal_r}}
     {targetBone {boneName=pinky_metacarpal_r}}
     {targetBone {boneName=ring_metacarpal_r}}
     {targetBone {boneName=thumb_01_r}}
     {targetBone {boneName=deform_wrist_up_r}}
     {targetBone {boneName=lowerarm_twist_01_r}}
     {targetBone {boneName=lowerarm_twist_02_r}}
    ]
    constraintWeights [1,1,1,1,1,1,1,1]
   }
   n83 = "Constraint - dyn_wristband_03 (PPPPPPPP)"!(
    ComponentPose=@n82
    Alpha=1
    ConstraintWeights_0=0.0123
    ConstraintWeights_1:float=0.1101
    ConstraintWeights_2:float=0.054
    ConstraintWeights_3:float=0.0089
    ConstraintWeights_4:float=0.1413
    ConstraintWeights_5:float=0.191
    ConstraintWeights_6:float=0.4011
    ConstraintWeights_7:float=0.0814
   ) -> (Pose) {
    alphaBoolBlend {blendOption=Linear}
    boneToModify {boneName=dyn_wristband_03}
    constraintSetup [
     {targetBone {boneName=lowerarm_r}}
     {targetBone {boneName=hand_r}}
     {targetBone {boneName=pinky_metacarpal_r}}
     {targetBone {boneName=ring_metacarpal_r}}
     {targetBone {boneName=thumb_01_r}}
     {targetBone {boneName=deform_wrist_dn_r}}
     {targetBone {boneName=lowerarm_twist_01_r}}
     {targetBone {boneName=lowerarm_twist_02_r}}
    ]
    constraintWeights [1,1,1,1,1,1,1,1]
   }
   n84 = "Constraint - dyn_wristband_04 (PPPPPPPP)"!(
    ComponentPose=@n83
    Alpha=1
    ConstraintWeights_0=0.1745
    ConstraintWeights_1:float=0.1764
    ConstraintWeights_2:float=0.0091
    ConstraintWeights_3:float=0.0357
    ConstraintWeights_4:float=0.0117
    ConstraintWeights_5:float=0.1395
    ConstraintWeights_6:float=0.3896
    ConstraintWeights_7:float=0.0635
   ) -> (Pose) {
    alphaBoolBlend {blendOption=Linear}
    boneToModify {boneName=dyn_wristband_04}
    constraintSetup [
     {targetBone {boneName=hand_r}}
     {targetBone {boneName=pinky_metacarpal_r}}
     {targetBone {boneName=ring_metacarpal_r}}
     {targetBone {boneName=thumb_01_r}}
     {targetBone {boneName=deform_wrist_up_r}}
     {targetBone {boneName=deform_wrist_dn_r}}
     {targetBone {boneName=lowerarm_twist_01_r}}
     {targetBone {boneName=lowerarm_twist_02_r}}
    ]
    constraintWeights [1,1,1,1,1,1,1,1]
   }
   n85 = Control Rig!(
    Source=Component To Local(ComponentPose=@n84)
    Alpha
    moving 2d:double=Get moving2d()
    crouch pose:double=Get crouching() -> (crouching:double)
   ) -> (Pose) {
    alphaBoolBlend {blendOption=Linear}
    controlRigAssetReference {blueprintRigClass=&/M_MED_LazarusLens_Style_CtrlRig.M_MED_LazarusLens_Style_CtrlRig_C}
   }
   n86 = Get moving2d()
   n87 = "Transform (Modify) Bone - Bone: dyn_bk_cloth_MSTR"(ComponentPose=Local To Component(LocalPose=@n85), Alpha=@n86, Translation="0,0,0", Rotation="0,-15,0", Scale="1,1,1") {
    rotationMode=BMM_Additive
    rotationSpace=BCS_BoneSpace
    translationMode=BMM_Additive
    alphaBoolBlend {blendOption=Linear}
    alphaScaleBiasClamp {bInterpResult=true}
    boneToModify {boneName=dyn_bk_cloth_MSTR}
    rotation {yaw=-15}
   }
   n88 = "Transform (Modify) Bone - Bone: dyn_r_cloth_MSTR"(ComponentPose=@n87, Alpha=@n86, Translation="0,0,0", Rotation="0,-15,0", Scale="1,1,1") {
    rotationMode=BMM_Additive
    rotationSpace=BCS_BoneSpace
    translationMode=BMM_Additive
    alphaBoolBlend {blendOption=Linear}
    alphaScaleBiasClamp {bInterpResult=true}
    boneToModify {boneName=dyn_r_cloth_MSTR}
    rotation {yaw=-15}
   }
   n89 = "Transform (Modify) Bone - Bone: dyn_l_cloth_MSTR"(ComponentPose=@n88, Alpha=@n86, Translation="0,0,0", Rotation="0,-15,0", Scale="1,1,1") {
    rotationMode=BMM_Additive
    rotationSpace=BCS_BoneSpace
    translationMode=BMM_Additive
    alphaBoolBlend {blendOption=Linear}
    alphaScaleBiasClamp {bInterpResult=true}
    boneToModify {boneName=dyn_l_cloth_MSTR}
    rotation {yaw=-15}
   }
   n90 = "Transform (Modify) Bone - Bone: dyn_jacket_bk_MSTR"(ComponentPose=@n89, Alpha=@n86, Translation="0,0,0", Rotation="0,-15,0", Scale="1,1,1") {
    rotationMode=BMM_Additive
    rotationSpace=BCS_BoneSpace
    translationMode=BMM_Additive
    alphaBoolBlend {blendOption=Linear}
    alphaScaleBiasClamp {bInterpResult=true}
    boneToModify {boneName=dyn_jacket_bk_MSTR}
    rotation {yaw=-15}
   }
   n91 = "Transform (Modify) Bone - Bone: Floor_COL"(ComponentPose=@n90, Alpha=Get jumping() -> (jumping:double), Translation="0,0,-100", Rotation="0,0,0", Scale="1,1,1") {
    rotationMode=BMM_Additive
    rotationSpace=BCS_BoneSpace
    translationMode=BMM_Additive
    alphaBoolBlend {blendOption=Linear}
    alphaScaleBiasClamp {bInterpResult=true}
    boneToModify {boneName=Floor_COL}
    translation {z=-100}
   }
   Save cached pose 'pre rigid'(Pose=Component To Local(ComponentPose=@n91)) {cacheName=pre rigid}
  ]}
 EventGraph {nodes [
   Event BlueprintUpdateAnimation()
   > n1 = Sequence!(execute) -> (then_0, then_1, then_2:Exec, then_3:Exec, then_4:Exec, then_5:Exec)
   > n2 = Gravity Override State Function DEPRECATED(CharacterPartAnimInstance=Self-Reference(), GameTimeInSeconds=GetGameTimeInSeconds(), GravityOverrideParams=Get Gravity Override Params())
   > Set Gravity Override(Gravity Override=@n2)
   > Set moving2d(moving2d:double="To Float (Boolean)"(InBool=Get bIsMoving2D())) -> (Output_Get:double)
   > Set crouching(crouching:double="To Float (Boolean)"(InBool=Get bIsCrouching())) -> (Output_Get:double)
   > Branch(Condition=CheckForBackbling())
   > n3 = Gravity Override State Function DEPRECATED(
    CharacterPartAnimInstance=Self-Reference()
    GameTimeInSeconds=GetGameTimeInSeconds()
    GravityOverrideParams=Get Gravity Override Params_Pack() -> (Gravity Override Params_Pack:Gravity Override Params Struct Structure)
   )
   > Set Gravity Override(Gravity Override=@n3)
   n4 = Branch(execute=@n1.then_1, Condition=Get bIsCrouching())
   Set Sim Space Settings(
    execute=[@n5.then, @n4.then]
    Sim Space Settings:Sim Space Settings Structure=Get Sim Space Settings_crouch() -> (Sim Space Settings_crouch:Sim Space Settings Structure)
   ) -> (Output_Get:Sim Space Settings Structure)
   Set Sim Space Settings(execute=[@n5.else, @n4.else], Sim Space Settings=Get Sim Space Settings_stand() -> (Sim Space Settings_stand:Sim Space Settings Structure))
   n5 = Branch(execute=@n1.then_4, Condition=CheckForBackbling())
   n6 = Branch(execute=@n1.then_5, Condition=CheckForBackbling())
   > Set backbling_cloth(backbling_cloth:double=0.1) -> (Output_Get:double)
   Set backbling_cloth(execute=@n6.else, backbling_cloth=0)
   n7 = BreakTransform(InTransform=GetDeltaTransformFromRefPose(self=GetOwnerSkeletalMeshComponent(), BoneName=dyn_necklace_c_deform_l, BaseName=spine_04))
   Set pec l(execute=@n1.then_2, pec l:double=MapRangeClamped(Value=BreakVector(InVec=@n7).Z, InRangeA=-10, InRangeB=-100)) -> (Output_Get:double)
   > Set Arm L rot(Arm L rot:double=GetAlphaFromReferenceTransformDelta(SkeletalMeshComponent=Get Skel_Mesh(), BoneName=clavicle_l, BaseName=spine_05, InputRange=(X=2,Y=7), TwistAxis="1,0,0")) -> (Output_Get:double)
   > n8 = Branch(Condition=Get bIsCrouching())
   > n9 = PlayerMovement(AnimInstance=Self-Reference(), XDotProductCoefficient=1, YDotProductCoefficient=1, ZDotProductCoefficient=1)
   > Set strafe right(strafe right:double=MapRangeClamped(Value=@n9, InRangeA=0, InRangeB=1)) -> (Output_Get:double)
   n10 = Get Skel_Mesh()
   > Set ThighRotL(ThighRotL:double=float * float(
     A=GetAlphaFromReferenceTransformDelta(SkeletalMeshComponent=@n10, BoneName=thigh_l, BaseName=pelvis, InputRange=(X=-20,Y=-100), TwistAxis="1,0,0")
     B=Get strafe right() -> (strafe right:double)
    ) {is=title K2Node_PromotableOperator}) -> (Output_Get:double)
   Set ThighRotL(execute=@n8.else, ThighRotL=0)
   n11 = Get bIsMoving2D()
   n12 = Branch(execute=@n1.then_3, Condition=AND Boolean(A=Get bIsCrouching(), B=NOT Boolean(A=@n11)))
   > Set crouch pose(crouch pose:double=1) -> (Output_Get:double)
   Set crouch pose(execute=@n12.else, crouch pose=0)
   n13 = Event OnBeginFalling()
   Set jumping(execute=[@n13.then, @n15.then, @n16.then], jumping:double=1) -> (Output_Get:double)
   n14 = Event OnEndFalling()
   Set jumping(execute=[@n14.then, @n17.then, @n18.then], jumping=0)
   Event BlueprintInitializeAnimation()
   > Set Skel_Mesh(Skel_Mesh=GetOwnerSkeletalMeshComponent())
   n15 = Event OnBeginJumping()
   n16 = Event OnBeginSliding() {is=title K2Node_Event}
   n17 = Event OnEndJumping()
   n18 = Event OnEndSliding() {is=title K2Node_Event}
   BreakRotator(InRot=@n7.Rotation)
   GetAlphaFromReferenceTransformDelta(
    SkeletalMeshComponent=Get Skel_Mesh()
    BoneName=deform_pec_aim_l
    BaseName=spine_04
    DeltaComponentType=EulerY
    InputRange=(X=0,Y=1)
    TwistAxis="1,0,0"
   )
  ]}
]
variables [
 Arm L rot {type=number}
 backbling_cloth {type=number}
 crouch pose {type=number}
 crouching {type=number}
 Gravity Override {type=Vector}
 Gravity Override Params {type=GravityOverrideParamsStruct}
 Gravity Override Params_Pack {
  type=GravityOverrideParamsStruct
  default {
   skydive_DivingDown {windAmplitude=<50 75 25>, windFrequency=<0.5 2 3>}
   skydive_FloatingDown {gravityOverride=<0 0 -540>, windAmplitude=<50 -150 100>, windFrequency=<0.5 1.5 2>, windOffset=<0 0 0>}
   skydive_GoingUp {gravityOverride=<0 0 -980>, windAmplitude=<50 -125 100>, windFrequency=<0.5 1.5 2>}
   skydive_Parachuting {gravityOverride=<0 0 -980>, windAmplitude=<50 -125 100>, windFrequency=<0.5 2 3>}
  }
 }
 jumping {type=number}
 moving2d {type=number}
 pec l {type=number}
 Sim Space Settings {
  type=SimSpaceSettings
  default {
   dampingAlpha=1
   maxAngularAcceleration=500
   maxAngularVelocity=3
   maxLinearAcceleration=350
   maxLinearVelocity=250
   velocityScaleZ=0
   worldAlpha=0.3
   externalLinearDragV=<0.2 0.2 0>
  }
 }
 Sim Space Settings_crouch {
  type=SimSpaceSettings
  default {
   dampingAlpha=1
   maxAngularAcceleration=500
   maxAngularVelocity=3
   maxLinearAcceleration=350
   maxLinearVelocity=250
   velocityScaleZ=0
   worldAlpha=0.2
   externalLinearDragV=<0.2 0.2 0>
  }
 }
 Sim Space Settings_stand {
  type=SimSpaceSettings
  default {
   dampingAlpha=1
   maxAngularAcceleration=1000
   maxAngularVelocity=3
   maxLinearAcceleration=700
   maxLinearVelocity=500
   velocityScaleZ=0
   worldAlpha=0.7
   externalLinearDragV=<0.2 0.2 0>
  }
 }
 Skel_Mesh {type=SkeletalMeshComponent}
 strafe right {type=number}
 ThighRotL {type=number}
]
