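# Physics asset for the LazarusLens style male medium character's face accessory mesh, covering three dynamic hair chains: back, left and right.
# Capsules on the upper spine, clavicles and upper arms form body volumes for the hair to collide with.
# Each chain attaches to spine_05 through a master bone; its upper joints have tiny capsules and free rotation.
# Lower segments carry larger offset capsules and tight, asymmetric swing and twist limits.
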PhysicsAsset /BRCosmetics/Characters/Player/Male/Medium/Bodies/M_MED_LazarusLens_Style/Meshes/Parts/M_MED_LazarusLens_Style_FaceAcc_Physics
bodies {
 clavicle_l [{shapeType=Capsule, length=12.1784, radius=8.3362, shapeName=clavicle_r_capsule, center=<6.832 -2.2208 0.3284>, rotation=<rot 90 180 -175>}]
 clavicle_r [{
   shapeType=Capsule
   length=12.1784
   radius=8.3362
   shapeName=clavicle_r_capsule
   center=<-6.9394 1.8911 -0.3284>
   rotation=<rot 90 -179.1048 -174.1049>
  }]
 dyn_Hair_Back_01 [{shapeType=Capsule, length=0.505, radius=0.505, shapeName=dyn_Hair_Back_01_capsule}]
 dyn_Hair_Back_02 [{shapeType=Capsule, length=0.505, radius=0.505, shapeName=dyn_Hair_Back_02_capsule}]
 dyn_Hair_Back_03 [{shapeType=Capsule, length=2.5, radius=3.125, shapeName=dyn_Hair_Back_03_capsule, center=<5 -2 0>, rotation=<rot 90 0 0>}]
 dyn_Hair_Back_04 [{shapeType=Capsule, length=2.5, radius=3.125, shapeName=dyn_Hair_Back_04_capsule, center=<5 -2 0>, rotation=<rot 90 0 0>}]
 dyn_Hair_Back_05 [{shapeType=Capsule, length=1.625, radius=4, shapeName=dyn_Hair_Back_05_capsule, center=<5 -2 0>, rotation=<rot 90 -72.1574 -87.1574>}]
 dyn_Hair_Back_MSTR [{shapeType=Capsule, length=0.505, radius=0.505, shapeName=dyn_Hair_Back_MSTR_capsule}]
 dyn_Hair_Left_01 [{shapeType=Capsule, length=0.505, radius=0.505, shapeName=dyn_Hair_BackL_01_capsule}]
 dyn_Hair_Left_02 [{shapeType=Capsule, length=0.505, radius=0.505, shapeName=dyn_Hair_BackL_02_capsule}]
 dyn_Hair_Left_03 [{shapeType=Capsule, length=2.5, radius=3.125, shapeName=dyn_Hair_BackL_03_capsule, center=<5 -2 0>, rotation=<rot 90 0 0>}]
 dyn_Hair_Left_04 [{shapeType=Capsule, length=2.5, radius=3.125, shapeName=dyn_Hair_BackL_04_capsule, center=<5 -2 0>, rotation=<rot 90 0 0>}]
 dyn_Hair_Left_MSTR [{shapeType=Capsule, length=0.505, radius=0.505, shapeName=dyn_Hair_BackL_MSTR_capsule}]
 dyn_Hair_Right_01 [{shapeType=Capsule, length=0.505, radius=0.505, shapeName=dyn_Hair_BackR_01_capsule}]
 dyn_Hair_Right_02 [{shapeType=Capsule, length=0.505, radius=0.505, shapeName=dyn_Hair_BackR_02_capsule}]
 dyn_Hair_Right_03 [
  {shapeType=Capsule, length=2.5, radius=3.125, shapeName=dyn_Hair_BackR_03_capsule, center=<5 -2 0>, rotation=<rot 90 0 0>}
  {
   shapeType=Capsule
   length=1.375
   radius=2.625
   shapeName=dyn_Hair_BackR_03_capsule
   center=<8.6648 -2.2662 -4.6557>
   rotation=<rot 70 -19.9998 0>
  }
 ]
 dyn_Hair_Right_04 [{shapeType=Capsule, length=2, radius=3.625, shapeName=dyn_Hair_BackR_04_capsule, center=<5 -2 0>, rotation=<rot 90 0 0>}]
 dyn_Hair_Right_MSTR [{shapeType=Capsule, length=0.505, radius=0.505, shapeName=dyn_Hair_BackR_MSTR_capsule}]
 spine_04 [{shapeType=Capsule, length=20.2816, radius=12.0126, shapeName=spine_04_capsule, center=<5.6062 -1.9258 0>, rotation=<rot 0 -92.6847 0>}]
 spine_05 [{shapeType=Capsule, length=12.5661, radius=11.4576, shapeName=spine_05_capsule, center=<1.7697 -1.1977 -0.1551>}]
 upperarm_l [{shapeType=Capsule, length=14.0479, radius=6.7679, shapeName=upperarm_l_capsule, center=<5 -1.25 0.0001>, rotation=<rot 0 -90 90>}]
 upperarm_r [{shapeType=Capsule, length=11.0479, radius=7.2679, shapeName=upperarm_r_capsule, center=<-5.5837 1.25 0.0001>, rotation=<rot 0 -90 90>}]
}
constraints [
 {
  bone1Name=spine_05
  bone2Name=spine_04
  swing1LimitDegrees=45
  swing1Motion=Free
  swing2LimitDegrees=45
  swing2Motion=Free
  twistLimitDegrees=45
  twistMotion=Free
 }
 {
  bone1Name=clavicle_l
  bone2Name=spine_05
  swing1LimitDegrees=45
  swing1Motion=Free
  swing2LimitDegrees=45
  swing2Motion=Free
  twistLimitDegrees=45
  twistMotion=Free
 }
 {
  bone1Name=upperarm_l
  bone2Name=clavicle_l
  swing1LimitDegrees=45
  swing1Motion=Free
  swing2LimitDegrees=45
  swing2Motion=Free
  twistLimitDegrees=45
  twistMotion=Free
 }
 {
  bone1Name=clavicle_r
  bone2Name=spine_05
  swing1LimitDegrees=45
  swing1Motion=Free
  swing2LimitDegrees=45
  swing2Motion=Free
  twistLimitDegrees=45
  twistMotion=Free
 }
 {
  bone1Name=upperarm_r
  bone2Name=clavicle_r
  swing1LimitDegrees=45
  swing1Motion=Free
  swing2LimitDegrees=45
  swing2Motion=Free
  twistLimitDegrees=45
  twistMotion=Free
 }
 {
  bone1Name=dyn_Hair_Back_MSTR
  bone2Name=spine_05
  swing1LimitDegrees=45
  swing1Motion=Free
  swing2LimitDegrees=45
  swing2Motion=Free
  twistLimitDegrees=45
  twistMotion=Free
 }
 {
  bone1Name=dyn_Hair_Back_01
  bone2Name=dyn_Hair_Back_MSTR
  swing1LimitDegrees=45
  swing1Motion=Free
  swing2LimitDegrees=45
  swing2Motion=Free
  twistLimitDegrees=45
  twistMotion=Free
 }
 {
  bone1Name=dyn_Hair_Back_02
  bone2Name=dyn_Hair_Back_01
  swing1LimitDegrees=45
  swing1Motion=Free
  swing2LimitDegrees=45
  swing2Motion=Free
  twistLimitDegrees=45
  twistMotion=Free
 }
 {bone1Name=dyn_Hair_Back_03, bone2Name=dyn_Hair_Back_02, swing1LimitDegrees=35, swing2LimitDegrees=10, twistLimitDegrees=5}
 {bone1Name=dyn_Hair_Back_04, bone2Name=dyn_Hair_Back_03, swing1LimitDegrees=35, swing2LimitDegrees=10, twistLimitDegrees=5}
 {bone1Name=dyn_Hair_Back_05, bone2Name=dyn_Hair_Back_04, swing1LimitDegrees=35, swing2LimitDegrees=10, twistLimitDegrees=5}
 {
  bone1Name=dyn_Hair_Left_MSTR
  bone2Name=spine_05
  swing1LimitDegrees=45
  swing1Motion=Free
  swing2LimitDegrees=45
  swing2Motion=Free
  twistLimitDegrees=45
  twistMotion=Free
 }
 {
  bone1Name=dyn_Hair_Left_01
  bone2Name=dyn_Hair_Left_MSTR
  swing1LimitDegrees=45
  swing1Motion=Free
  swing2LimitDegrees=45
  swing2Motion=Free
  twistLimitDegrees=45
  twistMotion=Free
 }
 {
  bone1Name=dyn_Hair_Left_02
  bone2Name=dyn_Hair_Left_01
  swing1LimitDegrees=45
  swing1Motion=Free
  swing2LimitDegrees=45
  swing2Motion=Free
  twistLimitDegrees=45
  twistMotion=Free
 }
 {bone1Name=dyn_Hair_Left_03, bone2Name=dyn_Hair_Left_02, swing1LimitDegrees=35, swing2LimitDegrees=10, twistLimitDegrees=5}
 {bone1Name=dyn_Hair_Left_04, bone2Name=dyn_Hair_Left_03, swing1LimitDegrees=35, swing2LimitDegrees=10, twistLimitDegrees=5}
 {
  bone1Name=dyn_Hair_Right_MSTR
  bone2Name=spine_05
  swing1LimitDegrees=45
  swing1Motion=Free
  swing2LimitDegrees=45
  swing2Motion=Free
  twistLimitDegrees=45
  twistMotion=Free
 }
 {
  bone1Name=dyn_Hair_Right_01
  bone2Name=dyn_Hair_Right_MSTR
  swing1LimitDegrees=45
  swing1Motion=Free
  swing2LimitDegrees=45
  swing2Motion=Free
  twistLimitDegrees=45
  twistMotion=Free
 }
 {
  bone1Name=dyn_Hair_Right_02
  bone2Name=dyn_Hair_Right_01
  swing1LimitDegrees=45
  swing1Motion=Free
  swing2LimitDegrees=45
  swing2Motion=Free
  twistLimitDegrees=45
  twistMotion=Free
 }
 {bone1Name=dyn_Hair_Right_03, bone2Name=dyn_Hair_Right_02, swing1LimitDegrees=35, swing2LimitDegrees=10, twistLimitDegrees=5}
 {bone1Name=dyn_Hair_Right_04, bone2Name=dyn_Hair_Right_03, swing1LimitDegrees=35, swing2LimitDegrees=10, twistLimitDegrees=5}
]
