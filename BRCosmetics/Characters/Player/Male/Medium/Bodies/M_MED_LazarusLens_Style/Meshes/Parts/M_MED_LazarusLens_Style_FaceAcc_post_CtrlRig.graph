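# Post-process Control Rig for the LazarusLens_Style male medium face accessory that adjusts and animates its hair bones.
# Three spherical pose readers on the head bone store look up, left and right alphas.
# The look-up alpha then rotates one extra right-hair bone, before wave-only chain harmonics, scaled by the skydive variable, sway the back, left and right chains.
# Other look-up and look-right hair offsets are disconnected from execution.
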ControlRigBlueprint /BRCosmetics/Characters/Player/Male/Medium/Bodies/M_MED_LazarusLens_Style/Meshes/Parts/M_MED_LazarusLens_Style_FaceAcc_post_CtrlRig
graphs [
 RigVMFunctionLibrary {links [], nodes []}
 RigVMModel {nodes [
   n1 = RigVMVariableNode() {variable=head_SPR_look_up_alpha}
   &RigUnit_ModifyTransforms(
    ExecutePin=()
    ItemToModify="((Item=(Type=Bone,Name=\"dyn_Hair_Right_MSTR\"),Transform=(Rotation=(X=0,Y=0,Z=0,W=1),Translation=(X=0,Y=-5,Z=0),Scale3D=(X=1,Y=1,Z=1))),(Item=(Type=Bone,Name=\"dyn_Hair_Right_03\"),Transform=(Rotation=(X=0,Y=0,Z=0,W=1),Translation=(X=0,Y=1,Z=0),Scale3D=(X=1,Y ...(truncated)"
    WeightMinimum=0
    ItemToModify.2.Transform.Rotation=&RigVMFunction_MathQuaternionFromEuler(Euler=(X=0,Y=0,Z=15), RotationOrder=XYZ).Result
    ItemToModify.1.Transform.Rotation=&RigVMFunction_MathQuaternionFromEuler(Euler=(X=0,Y=0,Z=7), RotationOrder=XYZ).Result
    ItemToModify.0.Transform.Rotation=&RigVMFunction_MathQuaternionFromEuler(Euler=(X=0,Y=0,Z=-25), RotationOrder=XYZ).Result
    Weight=@n1.Value
   )
   > &RigUnit_ModifyTransforms(
    ItemToModify="((Item=(Type=Bone,Name=\"dyn_Hair_Left_MSTR\"),Transform=(Rotation=(X=0,Y=0,Z=0,W=1),Translation=(X=2,Y=-3,Z=0),Scale3D=(X=1,Y=1,Z=1))),(Item=(Type=Bone,Name=\"dyn_Hair_Left_03\"),Transform=(Rotation=(X=0,Y=0,Z=0,W=1),Translation=(X=1,Y=0,Z=0),Scale3D=(X=1,Y=1 ...(truncated)"
    WeightMinimum=0
    ItemToModify.1.Transform.Rotation=&RigVMFunction_MathQuaternionFromEuler(Euler=(X=0,Y=0,Z=15), RotationOrder=XYZ).Result
    ItemToModify.0.Transform.Rotation=&RigVMFunction_MathQuaternionFromEuler(Euler=(X=0,Y=0,Z=-15), RotationOrder=XYZ).Result
    Weight=@n1.Value
   )
   RigUnit_BeginExecution = &RigUnit_BeginExecution()
   SequenceAggregate = &RigVMFunction_Sequence(ExecuteContext=@RigUnit_BeginExecution.ExecutePin)
   n2 = &RigUnit_SphericalPoseReader(
    ActiveRegionScaleFactors=(PositiveWidth=1,NegativeWidth=0.5,PositiveHeight=2,NegativeHeight=2)
    ActiveRegionSize=0.1
    Debug=(bDrawDebug=false,DebugScale=25,DebugSegments=20,DebugThickness=0.25)
    DriverAxis=(X=0,Y=1,Z=0)
    DriverItem=(Type=Bone,Name="head")
    FalloffRegionScaleFactors=(PositiveWidth=1,NegativeWidth=0.7,PositiveHeight=2,NegativeHeight=2)
    FalloffSize=0.25
    FlipWidthScaling=false
    OptionalParentItem=(Type=Bone)
    RotationOffset=(X=90,Y=0,Z=-25)
    ExecutePin=@SequenceAggregate.A
   )
   > RigVMVariableNode(Value=@n2.OutputParam) {variable=head_SPR_look_up_alpha}
   > n3 = &RigUnit_SphericalPoseReader(
    ActiveRegionScaleFactors=(PositiveWidth=0.75,NegativeWidth=0.75,PositiveHeight=1,NegativeHeight=0.5)
    ActiveRegionSize=0.1
    Debug=(bDrawDebug=true,DebugScale=25,DebugSegments=20,DebugThickness=0.25)
    DriverAxis=(X=0,Y=1,Z=0)
    DriverItem=(Type=Bone,Name="head")
    FalloffRegionScaleFactors=(PositiveWidth=1,NegativeWidth=1,PositiveHeight=1,NegativeHeight=0.7)
    FalloffSize=0.2
    OptionalParentItem=(Type=Bone)
    RotationOffset=(X=120,Y=0,Z=0)
   )
   > RigVMVariableNode(Value=@n3.OutputParam) {variable=head_SPR_look_left_alpha}
   > n4 = &RigUnit_SphericalPoseReader(
    ActiveRegionScaleFactors=(PositiveWidth=0.75,NegativeWidth=0.75,PositiveHeight=1,NegativeHeight=0.5)
    ActiveRegionSize=0.1
    Debug=(bDrawDebug=true,DebugScale=25,DebugSegments=20,DebugThickness=0.25)
    DriverAxis=(X=0,Y=1,Z=0)
    DriverItem=(Type=Bone,Name="head")
    FalloffRegionScaleFactors=(PositiveWidth=1,NegativeWidth=1,PositiveHeight=1,NegativeHeight=0.7)
    FalloffSize=0.2
    FlipHeightScaling=true
    FlipWidthScaling=false
    OptionalParentItem=(Type=Bone)
    RotationOffset=(X=60,Y=0,Z=0)
   )
   > RigVMVariableNode(Value=@n4.OutputParam) {variable=head_SPR_look_right_alpha}
   &RigUnit_ModifyTransforms(
    ItemToModify=((Item=(Type=Bone,Name="dyn_Hair_extra_08"),Transform=(Rotation=(X=0,Y=0,Z=0,W=1),Translation=(X=0,Y=0,Z=0),Scale3D=(X=1,Y=1,Z=1))))
    WeightMinimum=0
    ItemToModify.0.Transform.Rotation=&RigVMFunction_MathQuaternionFromEuler(Euler=(X=0,Y=0,Z=-25), RotationOrder=XYZ).Result
    ExecutePin=@SequenceAggregate.B
    Weight=@n1.Value
   )
   n5 = RigVMVariableNode() {variable=skydive}
   > /Script/ControlRig.RigUnit_ChainHarmonicsPerItem(
    ChainRoot=(Type=Bone,Name="dyn_Hair_Back_01")
    DrawWorldOffset=(Rotation=(X=0,Y=0,Z=0,W=1),Translation=(X=0,Y=0,Z=0),Scale3D=(X=1,Y=1,Z=1))
    Pendulum=(bEnabled=False,PendulumStiffness=2,PendulumGravity=(X=0,Y=0,Z=0),PendulumBlend=0.75,PendulumDrag=0.98,PendulumMinimum=0,PendulumMaximum=0.1,PendulumEase=Linear,UnwindAxis=(X=0,Y=1,Z=0),UnwindMinimum=0.2,UnwindMaximum=0.05)
    Reach=(bEnabled=true,ReachTarget=(X=0,Y=0,Z=0),ReachAxis=(X=1,Y=0,Z=0),ReachMinimum=0,ReachMaximum=0,ReachEase=Linear)
    Speed=(X=1,Y=1,Z=1.5)
    Wave=(bEnabled=True,WaveFrequency=(X=1,Y=0.6,Z=20),WaveAmplitude=(X=0,Y=0.3,Z=0.7),WaveOffset=(X=0,Y=1,Z=2),WaveNoise=(X=0,Y=0.5,Z=1),WaveMinimum=0,WaveMaximum=1,WaveEase=Linear)
    WaveCurve=(EditorCurveData=(Keys=((),(Time=1,Value=1)),DefaultValue=340282346638528859811704183484516925440,PreInfinityExtrap=RCCE_Constant,PostInfinityExtrap=RCCE_Constant),ExternalCurve=None)
    bDrawDebug=True
    Wave.WaveMaximum=@n5.Value
   ) {is=struct RigVMUnitNode}
   > &RigUnit_ChainHarmonicsPerItem(
    ChainRoot=(Type=Bone,Name="dyn_Hair_Left_01")
    DrawWorldOffset=(Rotation=(X=0,Y=0,Z=0,W=1),Translation=(X=0,Y=0,Z=0),Scale3D=(X=1,Y=1,Z=1))
    Pendulum=(bEnabled=False,PendulumStiffness=2,PendulumGravity=(X=0,Y=0,Z=0),PendulumBlend=0.75,PendulumDrag=0.98,PendulumMinimum=0,PendulumMaximum=0.1,PendulumEase=Linear,UnwindAxis=(X=0,Y=1,Z=0),UnwindMinimum=0.2,UnwindMaximum=0.05)
    Reach=(bEnabled=true,ReachTarget=(X=0,Y=0,Z=0),ReachAxis=(X=1,Y=0,Z=0),ReachMinimum=0,ReachMaximum=0,ReachEase=Linear)
    Speed=(X=1,Y=1,Z=1.5)
    Wave=(bEnabled=True,WaveFrequency=(X=1,Y=0.6,Z=10),WaveAmplitude=(X=0,Y=0.1,Z=0.3),WaveOffset=(X=0,Y=0.5,Z=1),WaveNoise=(X=0,Y=0.25,Z=0.5),WaveMinimum=0,WaveMaximum=1,WaveEase=Linear)
    WaveCurve=(EditorCurveData=(Keys=((),(Time=1,Value=1)),DefaultValue=340282346638528859811704183484516925440,PreInfinityExtrap=RCCE_Constant,PostInfinityExtrap=RCCE_Constant),ExternalCurve=None)
    bDrawDebug=True
    Wave.WaveMaximum=@n5.Value
   )
   > &RigUnit_ChainHarmonicsPerItem(
    ChainRoot=(Type=Bone,Name="dyn_Hair_Right_01")
    DrawWorldOffset=(Rotation=(X=0,Y=0,Z=0,W=1),Translation=(X=0,Y=0,Z=0),Scale3D=(X=1,Y=1,Z=1))
    Pendulum=(bEnabled=False,PendulumStiffness=2,PendulumGravity=(X=0,Y=0,Z=0),PendulumBlend=0.75,PendulumDrag=0.98,PendulumMinimum=0,PendulumMaximum=0.1,PendulumEase=Linear,UnwindAxis=(X=0,Y=1,Z=0),UnwindMinimum=0.2,UnwindMaximum=0.05)
    Reach=(bEnabled=true,ReachTarget=(X=0,Y=0,Z=0),ReachAxis=(X=1,Y=0,Z=0),ReachMinimum=0,ReachMaximum=0,ReachEase=Linear)
    Speed=(X=1,Y=1,Z=1.5)
    Wave=(bEnabled=True,WaveFrequency=(X=1,Y=0.6,Z=20),WaveAmplitude=(X=0,Y=0.3,Z=0.7),WaveOffset=(X=0,Y=1,Z=2),WaveNoise=(X=0,Y=0.5,Z=1),WaveMinimum=0,WaveMaximum=1,WaveEase=Linear)
    WaveCurve=(EditorCurveData=(Keys=((),(Time=1,Value=1)),DefaultValue=340282346638528859811704183484516925440,PreInfinityExtrap=RCCE_Constant,PostInfinityExtrap=RCCE_Constant),ExternalCurve=None)
    bDrawDebug=True
    Wave.WaveMaximum=@n5.Value
   )
   &RigUnit_ModifyTransforms(
    ExecutePin=()
    ItemToModify="((Item=(Type=Bone,Name=\"dyn_Hair_Right_MSTR\"),Transform=(Rotation=(X=0,Y=0,Z=0,W=1),Translation=(X=0,Y=-3,Z=0),Scale3D=(X=1,Y=1,Z=1))),(Item=(Type=Bone,Name=\"dyn_Hair_Right_03\"),Transform=(Rotation=(X=0,Y=0,Z=0,W=1),Translation=(X=0,Y=1,Z=0),Scale3D=(X=1,Y ...(truncated)"
    WeightMinimum=0
    ItemToModify.1.Transform.Rotation=&RigVMFunction_MathQuaternionFromEuler(Euler=(X=0,Y=0,Z=3), RotationOrder=XYZ).Result
    ItemToModify.2.Transform.Rotation=&RigVMFunction_MathQuaternionFromEuler(Euler=(X=0,Y=0,Z=7), RotationOrder=XYZ).Result
    ItemToModify.0.Transform.Rotation=&RigVMFunction_MathQuaternionFromEuler(Euler=(X=0,Y=0,Z=-10), RotationOrder=XYZ).Result
    Weight=RigVMVariableNode() {variable=head_SPR_look_right_alpha}.Value
   )
   &RigVMFunction_MathQuaternionFromEuler(Euler=(X=0,Y=0,Z=10), RotationOrder=XYZ)
   RigVMVariableNode() {variable=head_SPR_look_left_alpha}
  ]}
]
hierarchy [
 root {loc=<0 0 0>, rot=<0 0 0 1>}
 pelvis {parent=root, loc=<0 -0.7071 93.1191>, rot=<0.0127 0.707 -0.0127 -0.707>}
 spine_01 {parent=pelvis, loc=<6.7494 -0.6388 0>, rot=<0 0 0.034 -0.9994>}
 spine_02 {parent=spine_01, loc=<7.2518 -0.035 0>, rot=<0 0 0.0238 -0.9997>}
 spine_03 {parent=spine_02, loc=<6.0104 0 0>, rot=<0 0 0.0149 -0.9999>}
 spine_04 {parent=spine_03, loc=<8.1024 0 0>, rot=<0 0 -0.0132 -0.9999>}
 spine_05 {parent=spine_04, loc=<10.0146 0 0>, rot=<0 0 -0.0015 -1>}
 neck_01 {parent=spine_05, loc=<11.3138 0.1529 0>, rot=<0 0 -0.2003 -0.9797>}
 neck_02 {parent=neck_01, loc=<4.5627 0 0>, rot=<0 0 0.0504 -0.9987>}
 head {parent=neck_02, loc=<5.3914 0 0>, rot=<0 0 0.115 -0.9934>}
 dyn_Hair_Back_MSTR {parent=head, loc=<4.5118 -6.4312 0.5844>, rot=<0.338 -0.9412 0 0>}
 dyn_Hair_Back_01 {parent=dyn_Hair_Back_MSTR, loc=<0 0 0>, rot=<0 0 0 1>}
 dyn_Hair_Back_02 {parent=dyn_Hair_Back_01, loc=<2.4962 0 0>, rot=<0 0 -0.1793 -0.9838>}
 dyn_Hair_Back_03 {parent=dyn_Hair_Back_02, loc=<4.385 0 0>, rot=<-0.0032 0.0118 0.0322 -0.9994>}
 dyn_Hair_Back_04 {parent=dyn_Hair_Back_03, loc=<6.8932 0 0>, rot=<-0.0029 0.0092 0.0068 -0.9999>}
 dyn_Hair_Back_05 {parent=dyn_Hair_Back_04, loc=<7.1689 0 0>, rot=<0 0 0 1>}
 dyn_Hair_extra_04 {parent=dyn_Hair_Back_02, loc=<2.1313 -4.7478 2.8695>, rot=<0.2199 0.0356 -0.1556 -0.9624>}
 dyn_Hair_Right_MSTR {parent=head, loc=<4.5118 -6.4312 5.3409>, rot=<0.338 -0.9412 0 0>}
 dyn_Hair_Right_01 {parent=dyn_Hair_Right_MSTR, loc=<0 0 0>, rot=<0 0 0 1>}
 dyn_Hair_Right_02 {parent=dyn_Hair_Right_01, loc=<1.9765 1.3782 0>, rot=<0 0 -0.1793 -0.9838>}
 dyn_Hair_Right_03 {parent=dyn_Hair_Right_02, loc=<4.3965 0 -0.4863>, rot=<-0.0032 0.0118 0.0322 -0.9994>}
 dyn_Hair_Right_04 {parent=dyn_Hair_Right_03, loc=<6.8875 -1.3554 -0.6961>, rot=<-0.0029 0.0092 0.0068 -0.9999>}
 dyn_Hair_extra_08 {parent=dyn_Hair_Right_03, loc=<-0.9108 2.4481 -2.1459>, rot=<-0.7374 -0.1519 -0.1169 -0.6476>}
 dyn_Hair_extra_07 {parent=dyn_Hair_Right_03, loc=<2.1727 -4.9112 -1.0483>, rot=<-0.3949 -0.0883 -0.1702 -0.8985>}
 dyn_Hair_extra_09 {parent=dyn_Hair_Right_02, loc=<1.6523 0.9662 -3.0122>, rot=<-0.7527 -0.1217 -0.1033 -0.6387>}
 dyn_Hair_Left_MSTR {parent=head, loc=<4.5118 -6.4312 -5.341>, rot=<0.338 -0.9412 0 0>}
 dyn_Hair_Left_01 {parent=dyn_Hair_Left_MSTR, loc=<0 0 0>, rot=<0 0 0 1>}
 dyn_Hair_Left_02 {parent=dyn_Hair_Left_01, loc=<1.977 1.378 0>, rot=<0 0 -0.1793 -0.9838>}
 dyn_Hair_Left_03 {parent=dyn_Hair_Left_02, loc=<4.397 0 0.486>, rot=<-0.0032 0.0118 0.0322 -0.9994>}
 dyn_Hair_Left_04 {parent=dyn_Hair_Left_03, loc=<6.8987 -1.3514 0.0648>, rot=<-0.0029 0.0092 0.0068 -0.9999>}
 dyn_Hair_extra_03 {parent=dyn_Hair_Left_03, loc=<-1.5351 -2.9922 2.6132>, rot=<0.5567 0.0983 -0.1646 -0.8083>}
 dyn_Hair_extra_05 {parent=dyn_Hair_Left_02, loc=<1.2428 0.2664 3.4335>, rot=<0.7316 0.1183 -0.1072 -0.6628>}
 dyn_Hair_extra_02 {parent=dyn_Hair_Left_01, loc=<0.2703 -1.2788 3.1463>, rot=<0.6612 0.2344 -0.2381 -0.6717>}
 dyn_knot {parent=head, loc=<15.1968 -3.2451 -1.1162>, rot=<0.0041 -1 0 0>}
 dyn_knot_b {parent=dyn_knot, loc=<-1.0635 -3.4934 2.1497>, rot=<0.153 0.153 0.6903 -0.6903>}
 dyn_knot_a {parent=dyn_knot, loc=<-2.6876 -3.7201 1.6561>, rot=<0.183 0.183 0.683 -0.683>}
 dyn_Hair_extra_01 {parent=head, loc=<9.0711 -9.2887 -6.9791>, rot=<0.0038 -0.9218 -0.3876 0.0016>}
 dyn_Hair_extra_06 {parent=head, loc=<8.2237 -6.4824 5.9237>, rot=<0.0028 -0.6664 0.7455 -0.0031>}
 clavicle_l {parent=spine_05, loc=<4.3575 -0.1466 -2.3223>, rot=<0.0465 -0.7289 -0.009 -0.6829>}
 upperarm_l {parent=clavicle_l, loc=<15.5111 0 0>, rot=<0.0226 -0.3234 -0.0088 -0.9459>}
 upperarm_twist_01_l {parent=upperarm_l, loc=<0.0041 0 0>, rot=<0 0 0 1>}
 deform_upperarm_bk_base_l {parent=upperarm_twist_01_l, loc=<1.0813 -7.2193 -1.9453>, rot=<0 0 0 1>}
 deform_upperarm_bk_l {parent=deform_upperarm_bk_base_l, loc=<0 0 0>, rot=<0 0 0 1>}
 deform_upperarm_up_base_l {parent=upperarm_twist_01_l, loc=<-3.2216 -1.1391 3.3311>, rot=<0 0 0 1>}
 deform_upperarm_up_l {parent=deform_upperarm_up_base_l, loc=<0 0 0>, rot=<0 0 0 1>}
 clavicle_r {parent=spine_05, loc=<4.3575 -0.1466 2.3223>, rot=<0.7289 0.0465 -0.6829 0.009>}
 upperarm_r {parent=clavicle_r, loc=<-15.5111 0 0>, rot=<0.0226 -0.3234 -0.0088 -0.9459>}
 upperarm_twist_01_r {parent=upperarm_r, loc=<-0.0041 0 0>, rot=<0 0 0 -1>}
 deform_upperarm_bk_base_r {parent=upperarm_twist_01_r, loc=<-1.0814 7.2192 1.9456>, rot=<0 0 0 1>}
 deform_upperarm_bk_r {parent=deform_upperarm_bk_base_r, loc=<0 0 0>, rot=<0 0 0 1>}
 deform_upperarm_up_base_r {parent=upperarm_twist_01_r, loc=<3.2214 1.1391 -3.3308>, rot=<0 0 0 1>}
 deform_upperarm_up_r {parent=deform_upperarm_up_base_r, loc=<0 0 0>, rot=<0 0 0 -1>}
 upperarm_twist_02_r {parent=upperarm_r, loc=<-13.7941 0 0>, rot=<0 0 0 -1>}
]
variables [
 head_SPR_look_left_alpha {type=double, default=0}
 head_SPR_look_right_alpha {type=double, default=0}
 head_SPR_look_up_alpha {type=double, default=0}
 skydive {type=double, default=0}
]
