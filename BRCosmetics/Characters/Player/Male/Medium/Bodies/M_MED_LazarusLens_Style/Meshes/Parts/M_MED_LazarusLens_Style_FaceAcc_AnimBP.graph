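# Animation Blueprint for the LazarusLens Style face accessory, adding secondary motion to its hair, knot and mask bones.
# The copied parent pose passes through hair constraints, a Control Rig and RigidBody with gravity override, then additive hair rotation while jumping or falling without backbling.
# AnimDynamics springs move knot and extra hair bones, the latter mostly pre-constrained to weighted body and hair bones.
# A mask offset and skydive-driven post Control Rig finish the pose.
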AnimBlueprint /BRCosmetics/Characters/Player/Male/Medium/Bodies/M_MED_LazarusLens_Style/Meshes/Parts/M_MED_LazarusLens_Style_FaceAcc_AnimBP
target_skeleton=&/M_MED_LazarusLens_Style_FaceAcc_Skeleton
graphs [
 AnimGraph {nodes [
   "Constraint - dyn_Hair_Back_05 (P)"(Alpha=1, ConstraintWeights_0=1) {
    alphaBoolBlend {blendOption=Linear}
    boneToModify {boneName=dyn_Hair_Back_05}
    constraintSetup [{targetBone {boneName=spine_05}}]
    constraintWeights [1]
   }
   n1 = "Constraint - dyn_Hair_Back_01 (PP)"!(
    ComponentPose=Local To Component(LocalPose=Copy Pose From Mesh() {bUseAttachedParent=true})
    Alpha=0.5
    ConstraintWeights_0=0.6
    ConstraintWeights_1:float=0.4
   ) -> (Pose) {
    alphaBoolBlend {blendOption=Linear}
    boneToModify {boneName=dyn_Hair_Back_01}
    constraintSetup [{targetBone {boneName=head}}, {targetBone {boneName=neck_01}}]
    constraintWeights [1,1]
   }
   n2 = "Constraint - dyn_Hair_Back_02 (PP)"!(ComponentPose=@n1, Alpha=0.7, ConstraintWeights_0=0.4, ConstraintWeights_1:float=0.6) -> (Pose) {
    alphaBoolBlend {blendOption=Linear}
    boneToModify {boneName=dyn_Hair_Back_02}
    constraintSetup [{targetBone {boneName=head}}, {targetBone {boneName=neck_01}}]
    constraintWeights [1,1]
   }
   n3 = "Constraint - dyn_Hair_Back_03 (PP)"!(ComponentPose=@n2, Alpha=1, ConstraintWeights_0=0.6, ConstraintWeights_1:float=0.4) -> (Pose) {
    alphaBoolBlend {blendOption=Linear}
    boneToModify {boneName=dyn_Hair_Back_03}
    constraintSetup [{targetBone {boneName=neck_01}}, {targetBone {boneName=spine_05}}]
    constraintWeights [1,1]
   }
   n4 = "Constraint - dyn_Hair_Back_04 (PP)"!(ComponentPose=@n3, Alpha=1, ConstraintWeights_0=0.4, ConstraintWeights_1:float=0.6) -> (Pose) {
    alphaBoolBlend {blendOption=Linear}
    boneToModify {boneName=dyn_Hair_Back_04}
    constraintSetup [{targetBone {boneName=neck_01}}, {targetBone {boneName=spine_05}}]
    constraintWeights [1,1]
   }
   n5 = "Constraint - dyn_Hair_Right_01 (PP)"!(ComponentPose=@n4, Alpha=0.5, ConstraintWeights_0=0.6, ConstraintWeights_1:float=0.4) -> (Pose) {
    alphaBoolBlend {blendOption=Linear}
    boneToModify {boneName=dyn_Hair_Right_01}
    constraintSetup [{targetBone {boneName=head}}, {targetBone {boneName=neck_01}}]
    constraintWeights [1,1]
   }
   n6 = "Constraint - dyn_Hair_Right_02 (PP)"!(ComponentPose=@n5, Alpha=0.7, ConstraintWeights_0=0.4, ConstraintWeights_1:float=0.6) -> (Pose) {
    alphaBoolBlend {blendOption=Linear}
    boneToModify {boneName=dyn_Hair_Right_02}
    constraintSetup [{targetBone {boneName=head}}, {targetBone {boneName=spine_05}}]
    constraintWeights [1,1]
   }
   n7 = "Constraint - dyn_Hair_Right_03 (PP)"!(ComponentPose=@n6, Alpha=1, ConstraintWeights_0=0.6, ConstraintWeights_1:float=0.4) -> (Pose) {
    alphaBoolBlend {blendOption=Linear}
    boneToModify {boneName=dyn_Hair_Right_03}
    constraintSetup [{targetBone {boneName=neck_01}}, {targetBone {boneName=spine_05}}]
    constraintWeights [1,1]
   }
   n8 = "Constraint - dyn_Hair_Right_04 (PP)"!(ComponentPose=@n7, Alpha=1, ConstraintWeights_0=0.4, ConstraintWeights_1:float=0.6) -> (Pose) {
    alphaBoolBlend {blendOption=Linear}
    boneToModify {boneName=dyn_Hair_Right_04}
    constraintSetup [{targetBone {boneName=neck_01}}, {targetBone {boneName=clavicle_r}}]
    constraintWeights [1,1]
   }
   n9 = "Constraint - dyn_Hair_Left_01 (PP)"!(ComponentPose=@n8, Alpha=0.5, ConstraintWeights_0=0.6, ConstraintWeights_1:float=0.4) -> (Pose) {
    alphaBoolBlend {blendOption=Linear}
    boneToModify {boneName=dyn_Hair_Left_01}
    constraintSetup [{targetBone {boneName=head}}, {targetBone {boneName=neck_01}}]
    constraintWeights [1,1]
   }
   n10 = "Constraint - dyn_Hair_Left_02 (PP)"!(ComponentPose=@n9, Alpha=0.7, ConstraintWeights_0=0.4, ConstraintWeights_1:float=0.6) -> (Pose) {
    alphaBoolBlend {blendOption=Linear}
    boneToModify {boneName=dyn_Hair_Left_02}
    constraintSetup [{targetBone {boneName=head}}, {targetBone {boneName=spine_05}}]
    constraintWeights [1,1]
   }
   n11 = "Constraint - dyn_Hair_Left_03 (PP)"!(ComponentPose=@n10, Alpha=1, ConstraintWeights_0=0.6, ConstraintWeights_1:float=0.4) -> (Pose) {
    alphaBoolBlend {blendOption=Linear}
    boneToModify {boneName=dyn_Hair_Left_03}
    constraintSetup [{targetBone {boneName=neck_01}}, {targetBone {boneName=spine_05}}]
    constraintWeights [1,1]
   }
   n12 = "Constraint - dyn_Hair_Left_04 (PP)"!(ComponentPose=@n11, Alpha=1, ConstraintWeights_0=0.4, ConstraintWeights_1:float=0.6) -> (Pose) {
    alphaBoolBlend {blendOption=Linear}
    boneToModify {boneName=dyn_Hair_Left_04}
    constraintSetup [{targetBone {boneName=neck_01}}, {targetBone {boneName=clavicle_l}}]
    constraintWeights [1,1]
   }
   n13 = Control Rig(Source=Component To Local(ComponentPose=@n12)) {
    alphaBoolBlend {blendOption=Linear}
    controlRigAssetReference {blueprintRigClass=&/M_MED_LazarusLens_Style_FaceAcc_CtrlRig.M_MED_LazarusLens_Style_FaceAcc_CtrlRig_C}
   }
   n14 = RigidBody!(
    ComponentPose=Local To Component(LocalPose=@n13)
    Alpha
    OverrideWorldGravity:Vector=Get GravityOverride()
    ExternalForce
    ComponentLinearAccScale:Vector="0.3,0.3,0.3"
    ComponentLinearVelScale:Vector="0.3,0.3,0.3"
   ) -> (Pose) {
    bClampLinearTranslationLimitToRefPose=true
    bOverrideWorldGravity=true
    overridePhysicsAsset=&/M_MED_LazarusLens_Style_FaceAcc_Physics
    simulationSpace=BaseBoneSpace
    alphaBoolBlend {blendOption=Linear}
    baseBoneRef {boneName=spine_05}
    componentAppliedLinearAccClamp=<2500 2500 2500>
    componentLinearAccScale=<0.3 0.3 0.5>
    componentLinearVelScale=<0.3 0.3 0.5>
    overrideWorldGravity {z=-800}
   }
   n15 = Get jumping() -> (jumping:double)
   n16 = "Transform (Modify) Bone - Bone: dyn_Hair_Back_02"(ComponentPose=@n14, Alpha=@n15, Translation="0,0,0", Rotation="0,-15,0", Scale="1,1,1") {
    rotationMode=BMM_Additive
    rotationSpace=BCS_BoneSpace
    alphaBoolBlend {blendOption=Linear}
    alphaScaleBiasClamp {bInterpResult=true, interpSpeedIncreasing=2}
    boneToModify {boneName=dyn_Hair_Back_02}
    rotation {yaw=-15}
   }
   n17 = "Transform (Modify) Bone - Bone: dyn_Hair_Back_03"(ComponentPose=@n16, Alpha=@n15, Translation="0,0,0", Rotation="0,-15,0", Scale="1,1,1") {
    rotationMode=BMM_Additive
    rotationSpace=BCS_BoneSpace
    alphaBoolBlend {blendOption=Linear}
    alphaScaleBiasClamp {bInterpResult=true, interpSpeedIncreasing=2}
    boneToModify {boneName=dyn_Hair_Back_03}
    rotation {yaw=-15}
   }
   n18 = "Transform (Modify) Bone - Bone: dyn_Hair_Back_04"(ComponentPose=@n17, Alpha=@n15, Translation="0,0,0", Rotation="0,-15,0", Scale="1,1,1") {
    rotationMode=BMM_Additive
    rotationSpace=BCS_BoneSpace
    alphaBoolBlend {blendOption=Linear}
    alphaScaleBiasClamp {bInterpResult=true, interpSpeedIncreasing=2}
    boneToModify {boneName=dyn_Hair_Back_04}
    rotation {yaw=-15}
   }
   n19 = "Transform (Modify) Bone - Bone: dyn_Hair_Right_02"(ComponentPose=@n18, Alpha=@n15, Translation="0,0,0", Rotation="0,-15,0", Scale="1,1,1") {
    rotationMode=BMM_Additive
    rotationSpace=BCS_BoneSpace
    alphaBoolBlend {blendOption=Linear}
    alphaScaleBiasClamp {bInterpResult=true, interpSpeedIncreasing=2}
    boneToModify {boneName=dyn_Hair_Right_02}
    rotation {yaw=-15}
   }
   n20 = "Transform (Modify) Bone - Bone: dyn_Hair_Right_03"(ComponentPose=@n19, Alpha=@n15, Translation="0,0,0", Rotation="0,-15,0", Scale="1,1,1") {
    rotationMode=BMM_Additive
    rotationSpace=BCS_BoneSpace
    alphaBoolBlend {blendOption=Linear}
    alphaScaleBiasClamp {bInterpResult=true, interpSpeedIncreasing=2}
    boneToModify {boneName=dyn_Hair_Right_03}
    rotation {yaw=-15}
   }
   n21 = "Transform (Modify) Bone - Bone: dyn_Hair_Right_04"(ComponentPose=@n20, Alpha=@n15, Translation="0,0,0", Rotation="0,-15,0", Scale="1,1,1") {
    rotationMode=BMM_Additive
    rotationSpace=BCS_BoneSpace
    alphaBoolBlend {blendOption=Linear}
    alphaScaleBiasClamp {bInterpResult=true, interpSpeedIncreasing=2}
    boneToModify {boneName=dyn_Hair_Right_04}
    rotation {yaw=-15}
   }
   n22 = "Transform (Modify) Bone - Bone: dyn_Hair_Left_02"(ComponentPose=@n21, Alpha=@n15, Translation="0,0,0", Rotation="0,-15,0", Scale="1,1,1") {
    rotationMode=BMM_Additive
    rotationSpace=BCS_BoneSpace
    alphaBoolBlend {blendOption=Linear}
    alphaScaleBiasClamp {bInterpResult=true, interpSpeedIncreasing=2}
    boneToModify {boneName=dyn_Hair_Left_02}
    rotation {yaw=-15}
   }
   n23 = "Transform (Modify) Bone - Bone: dyn_Hair_Left_03"(ComponentPose=@n22, Alpha=@n15, Translation="0,0,0", Rotation="0,-15,0", Scale="1,1,1") {
    rotationMode=BMM_Additive
    rotationSpace=BCS_BoneSpace
    alphaBoolBlend {blendOption=Linear}
    alphaScaleBiasClamp {bInterpResult=true, interpSpeedIncreasing=2}
    boneToModify {boneName=dyn_Hair_Left_03}
    rotation {yaw=-15}
   }
   n24 = "Transform (Modify) Bone - Bone: dyn_Hair_Left_04"(ComponentPose=@n23, Alpha=@n15, Translation="0,0,0", Rotation="0,-15,0", Scale="1,1,1") {
    rotationMode=BMM_Additive
    rotationSpace=BCS_BoneSpace
    alphaBoolBlend {blendOption=Linear}
    alphaScaleBiasClamp {bInterpResult=true, interpSpeedIncreasing=2}
    boneToModify {boneName=dyn_Hair_Left_04}
    rotation {yaw=-15}
   }
   n25 = Anim Dynamics - Bone: dyn_knot(ComponentPose=@n24, Alpha=1, ExternalForce="0,0,0") {
    angularDampingOverride=0.99
    angularSpringConstant=600
    bAngularSpring=true
    bGravityOverrideInSimSpace=true
    bOverrideAngularDamping=true
    simulationSpace=RootRelative
    alphaBoolBlend {blendOption=Linear}
    boundBone {boneName=dyn_knot}
    componentAppliedLinearAccClamp=<3500 3500 3500>
    componentLinearAccScale=<0.3 0.3 0.3>
    componentLinearVelScale=<0.3 0.3 3>
    physicsBodyDefinitions [{
      sphereCollisionRadius=0
      boundBone {boneName=dyn_knot}
      boxExtents=<25 20 20>
      constraintSetup {angularLimitsMax=<1 5 10>, angularLimitsMin=<-1 -5 -10>, angularTarget=<1 0 0>}
      localJointOffset=<2 0 0>
     }]
    relativeSpaceBone {boneName=root}
   }
   n26 = Anim Dynamics - Bone: dyn_knot_a(ComponentPose=@n25, Alpha=0.7, ExternalForce="0,0,0") {
    angularDampingOverride=0.9999
    angularSpringConstant=500
    bAngularSpring=true
    bGravityOverrideInSimSpace=true
    bOverrideAngularDamping=true
    simulationSpace=RootRelative
    alphaBoolBlend {blendOption=Linear}
    boundBone {boneName=dyn_knot_a}
    componentAppliedLinearAccClamp=<3500 3500 3500>
    componentLinearAccScale=<1.2 1.2 1.2>
    componentLinearVelScale=<1.2 1.2 1.2>
    physicsBodyDefinitions [{
      sphereCollisionRadius=0
      boundBone {boneName=dyn_knot_a}
      boxExtents=<15 10 10>
      constraintSetup {angularLimitsMax=<1 15 25>, angularLimitsMin=<-1 -15 -25>, angularTarget=<0 0 0>}
      localJointOffset=<3 0 0>
     }]
    relativeSpaceBone {boneName=root}
   }
   n27 = Anim Dynamics - Bone: dyn_knot_b(ComponentPose=@n26, Alpha=1, ExternalForce="0,0,0") {
    angularDampingOverride=0.9999
    angularSpringConstant=500
    bAngularSpring=true
    bGravityOverrideInSimSpace=true
    bOverrideAngularDamping=true
    simulationSpace=RootRelative
    alphaBoolBlend {blendOption=Linear}
    boundBone {boneName=dyn_knot_b}
    componentAppliedLinearAccClamp=<3500 3500 3500>
    componentLinearAccScale=<1.2 1.2 1.2>
    componentLinearVelScale=<1.2 1.2 1.2>
    physicsBodyDefinitions [{
      sphereCollisionRadius=0
      boundBone {boneName=dyn_knot_b}
      boxExtents=<15 10 10>
      constraintSetup {angularLimitsMax=<1 15 25>, angularLimitsMin=<-1 -15 -25>, angularTarget=<0 0 0>}
      localJointOffset=<3 0 0>
     }]
    relativeSpaceBone {boneName=root}
   }
   n28 = "Constraint - dyn_Hair_extra_01 (PPPPPPPP)"!(
    ComponentPose=@n27
    Alpha=1
    ConstraintWeights_0=0.0001
    ConstraintWeights_1:float=0.7995
    ConstraintWeights_2:float=0.0004
    ConstraintWeights_3:float=0.0007
    ConstraintWeights_4:float=0.118
    ConstraintWeights_5:float=0.0003
    ConstraintWeights_6:float=0.0006
    ConstraintWeights_7:float=0.0803
   ) -> (Pose) {
    alphaBoolBlend {blendOption=Linear}
    boneToModify {boneName=dyn_Hair_extra_01}
    constraintSetup [
     {targetBone {boneName=spine_05}}
     {targetBone {boneName=head}}
     {targetBone {boneName=dyn_Hair_Back_03}}
     {targetBone {boneName=dyn_Hair_Back_02}}
     {targetBone {boneName=dyn_Hair_Back_01}}
     {targetBone {boneName=dyn_Hair_Left_03}}
     {targetBone {boneName=dyn_Hair_Left_02}}
     {targetBone {boneName=dyn_Hair_Left_01}}
    ]
    constraintWeights [1,1,1,1,1,1,1,1]
   }
   n29 = "Constraint - dyn_Hair_extra_02 (PPPPPPPPPPP)"!(
    ComponentPose=@n28
    Alpha=1
    ConstraintWeights_0=0.0044
    ConstraintWeights_1:float=0.0003
    ConstraintWeights_2:float=0.0011
    ConstraintWeights_3:float=0.0058
    ConstraintWeights_4:float=0.3509
    ConstraintWeights_5:float=0
    ConstraintWeights_6:float=0.0001
    ConstraintWeights_7:float=0.0163
    ConstraintWeights_8:float=0.0027
    ConstraintWeights_9:float=0.0111
    ConstraintWeights_10:float=0.6073
   ) -> (Pose) {
    alphaBoolBlend {blendOption=Linear}
    boneToModify {boneName=dyn_Hair_extra_02}
    constraintSetup [
     {targetBone {boneName=spine_05}}
     {targetBone {boneName=clavicle_l}}
     {targetBone {boneName=neck_01}}
     {targetBone {boneName=neck_02}}
     {targetBone {boneName=head}}
     {targetBone {boneName=dyn_Hair_Back_03}}
     {targetBone {boneName=dyn_Hair_Back_02}}
     {targetBone {boneName=dyn_Hair_Back_01}}
     {targetBone {boneName=dyn_Hair_Left_03}}
     {targetBone {boneName=dyn_Hair_Left_02}}
     {targetBone {boneName=dyn_Hair_Left_01}}
    ]
    constraintWeights [1,1,1,1,1,1,1,1,1,1,1]
   }
   n30 = "Constraint - dyn_Hair_extra_03 (PPPPPPPPPPPP)"!(
    ComponentPose=@n29
    Alpha=1
    ConstraintWeights_0=0.0093
    ConstraintWeights_1:float=0.0007
    ConstraintWeights_2:float=0.0022
    ConstraintWeights_3:float=0.0112
    ConstraintWeights_4:float=0.1309
    ConstraintWeights_5:float=0.0007
    ConstraintWeights_6:float=0.0026
    ConstraintWeights_7:float=0.0081
    ConstraintWeights_8:float=0.0022
    ConstraintWeights_9:float=0.5475
    ConstraintWeights_10:float=0.0511
    ConstraintWeights_11:float=0.2334
   ) -> (Pose) {
    alphaBoolBlend {blendOption=Linear}
    boneToModify {boneName=dyn_Hair_extra_03}
    constraintSetup [
     {targetBone {boneName=spine_05}}
     {targetBone {boneName=clavicle_l}}
     {targetBone {boneName=neck_01}}
     {targetBone {boneName=neck_02}}
     {targetBone {boneName=head}}
     {targetBone {boneName=dyn_Hair_Back_03}}
     {targetBone {boneName=dyn_Hair_Back_02}}
     {targetBone {boneName=dyn_Hair_Back_01}}
     {targetBone {boneName=dyn_Hair_Left_04}}
     {targetBone {boneName=dyn_Hair_Left_03}}
     {targetBone {boneName=dyn_Hair_Left_02}}
     {targetBone {boneName=dyn_Hair_Left_01}}
    ]
    constraintWeights [1,1,1,1,1,1,1,1,1,1,1,1]
   }
   n31 = "Constraint - dyn_Hair_extra_04 (PPPPPPPPP)"!(
    ComponentPose=@n30
    Alpha=1
    ConstraintWeights_0=0.1376
    ConstraintWeights_1:float=0.0128
    ConstraintWeights_2:float=0.1373
    ConstraintWeights_3:float=0.1569
    ConstraintWeights_4:float=0.0555
    ConstraintWeights_5:float=0.0162
    ConstraintWeights_6:float=0.341
    ConstraintWeights_7:float=0.1254
    ConstraintWeights_8:float=0.0173
   ) -> (Pose) {
    alphaBoolBlend {blendOption=Linear}
    boneToModify {boneName=dyn_Hair_extra_04}
    constraintSetup [
     {targetBone {boneName=head}}
     {targetBone {boneName=dyn_Hair_Back_04}}
     {targetBone {boneName=dyn_Hair_Back_03}}
     {targetBone {boneName=dyn_Hair_Back_02}}
     {targetBone {boneName=dyn_Hair_Back_01}}
     {targetBone {boneName=dyn_Hair_Left_04}}
     {targetBone {boneName=dyn_Hair_Left_03}}
     {targetBone {boneName=dyn_Hair_Left_02}}
     {targetBone {boneName=dyn_Hair_Left_01}}
    ]
    constraintWeights [1,1,1,1,1,1,1,1,1]
   }
   n32 = "Constraint - dyn_Hair_extra_05 (PPPPPPPP)"!(
    ComponentPose=@n31
    Alpha=1
    ConstraintWeights_0=0.026
    ConstraintWeights_1:float=0.002
    ConstraintWeights_2:float=0.0049
    ConstraintWeights_3:float=0.0231
    ConstraintWeights_4:float=0.5968
    ConstraintWeights_5:float=0.1791
    ConstraintWeights_6:float=0.0421
    ConstraintWeights_7:float=0.1258
   ) -> (Pose) {
    alphaBoolBlend {blendOption=Linear}
    boneToModify {boneName=dyn_Hair_extra_05}
    constraintSetup [
     {targetBone {boneName=spine_05}}
     {targetBone {boneName=clavicle_l}}
     {targetBone {boneName=neck_01}}
     {targetBone {boneName=neck_02}}
     {targetBone {boneName=head}}
     {targetBone {boneName=dyn_Hair_Left_03}}
     {targetBone {boneName=dyn_Hair_Left_02}}
     {targetBone {boneName=dyn_Hair_Left_01}}
    ]
    constraintWeights [1,1,1,1,1,1,1,1]
   }
   n33 = "Constraint - dyn_Hair_extra_07 (PPPPPPPPP)"!(
    ComponentPose=@n32
    Alpha=1
    ConstraintWeights_0=0.047
    ConstraintWeights_1:float=0.0303
    ConstraintWeights_2:float=0.0789
    ConstraintWeights_3:float=0.0508
    ConstraintWeights_4:float=0.0243
    ConstraintWeights_5:float=0.1368
    ConstraintWeights_6:float=0.4615
    ConstraintWeights_7:float=0.1555
    ConstraintWeights_8:float=0.0148
   ) -> (Pose) {
    alphaBoolBlend {blendOption=Linear}
    boneToModify {boneName=dyn_Hair_extra_07}
    constraintSetup [
     {targetBone {boneName=head}}
     {targetBone {boneName=dyn_Hair_Back_04}}
     {targetBone {boneName=dyn_Hair_Back_03}}
     {targetBone {boneName=dyn_Hair_Back_02}}
     {targetBone {boneName=dyn_Hair_Back_01}}
     {targetBone {boneName=dyn_Hair_Right_04}}
     {targetBone {boneName=dyn_Hair_Right_03}}
     {targetBone {boneName=dyn_Hair_Right_02}}
     {targetBone {boneName=dyn_Hair_Right_01}}
    ]
    constraintWeights [1,1,1,1,1,1,1,1,1]
   }
   n34 = "Constraint - dyn_Hair_extra_08 (PPPPPPPPP)"!(
    ComponentPose=@n33
    Alpha=1
    ConstraintWeights_0=0.0437
    ConstraintWeights_1:float=0.0191
    ConstraintWeights_2:float=0.0911
    ConstraintWeights_3:float=0.3244
    ConstraintWeights_4:float=0.0013
    ConstraintWeights_5:float=0.0071
    ConstraintWeights_6:float=0.278
    ConstraintWeights_7:float=0.2352
    ConstraintWeights_8:float=0.0001
   ) -> (Pose) {
    alphaBoolBlend {blendOption=Linear}
    boneToModify {boneName=dyn_Hair_extra_08}
    constraintSetup [
     {targetBone {boneName=spine_05}}
     {targetBone {boneName=neck_01}}
     {targetBone {boneName=neck_02}}
     {targetBone {boneName=head}}
     {targetBone {boneName=dyn_Hair_Back_04}}
     {targetBone {boneName=dyn_Hair_Right_04}}
     {targetBone {boneName=dyn_Hair_Right_03}}
     {targetBone {boneName=dyn_Hair_Right_02}}
     {targetBone {boneName=dyn_Hair_Right_01}}
    ]
    constraintWeights [1,1,1,1,1,1,1,1,1]
   }
   n35 = "Constraint - dyn_Hair_extra_09 (PPPPPPPPPP)"!(
    ComponentPose=@n34
    Alpha=1
    ConstraintWeights_0=0.0083
    ConstraintWeights_1:float=0.0043
    ConstraintWeights_2:float=0.0338
    ConstraintWeights_3:float=0.5558
    ConstraintWeights_4:float=0.0001
    ConstraintWeights_5:float=0.0001
    ConstraintWeights_6:float=0.0013
    ConstraintWeights_7:float=0.0785
    ConstraintWeights_8:float=0.3005
    ConstraintWeights_9:float=0.0172
   ) -> (Pose) {
    alphaBoolBlend {blendOption=Linear}
    boneToModify {boneName=dyn_Hair_extra_09}
    constraintSetup [
     {targetBone {boneName=spine_05}}
     {targetBone {boneName=neck_01}}
     {targetBone {boneName=neck_02}}
     {targetBone {boneName=head}}
     {targetBone {boneName=dyn_Hair_Back_02}}
     {targetBone {boneName=dyn_Hair_Back_01}}
     {targetBone {boneName=dyn_Hair_Right_04}}
     {targetBone {boneName=dyn_Hair_Right_03}}
     {targetBone {boneName=dyn_Hair_Right_02}}
     {targetBone {boneName=dyn_Hair_Right_01}}
    ]
    constraintWeights [1,1,1,1,1,1,1,1,1,1]
   }
   n36 = Anim Dynamics - Bone: dyn_Hair_extra_01(ComponentPose=@n35, Alpha=1, ExternalForce="0,0,0") {
    angularDampingOverride=0.9999
    angularSpringConstant=100
    bAngularSpring=true
    bGravityOverrideInSimSpace=true
    bOverrideAngularDamping=true
    simulationSpace=RootRelative
    alphaBoolBlend {blendOption=Linear}
    boundBone {boneName=dyn_Hair_extra_01}
    componentAppliedLinearAccClamp=<3500 3500 3500>
    componentLinearAccScale=<1.2 1.2 1.2>
    componentLinearVelScale=<1.2 1.2 1.2>
    physicsBodyDefinitions [{
      sphereCollisionRadius=0
      boundBone {boneName=dyn_Hair_extra_01}
      boxExtents=<20 10 10>
      constraintSetup {angularLimitsMax=<1 25 15>, angularLimitsMin=<-1 -25 -15>, angularTarget=<1 0 0>}
      localJointOffset=<1 0 0>
     }]
    relativeSpaceBone {boneName=root}
   }
   n37 = Anim Dynamics - Bone: dyn_Hair_extra_02(ComponentPose=@n36, Alpha=1, ExternalForce="0,0,0") {
    angularDampingOverride=0.9999
    angularSpringConstant=100
    bAngularSpring=true
    bGravityOverrideInSimSpace=true
    bOverrideAngularDamping=true
    simulationSpace=RootRelative
    alphaBoolBlend {blendOption=Linear}
    boundBone {boneName=dyn_Hair_extra_02}
    componentAppliedLinearAccClamp=<3500 3500 3500>
    componentLinearAccScale=<1.2 1.2 1.2>
    componentLinearVelScale=<1.2 1.2 1.2>
    physicsBodyDefinitions [{
      sphereCollisionRadius=0
      boundBone {boneName=dyn_Hair_extra_02}
      boxExtents=<20 10 10>
      constraintSetup {angularLimitsMax=<1 15 2>, angularLimitsMin=<-1 -15 -25>, angularTarget=<1 0 0>}
      localJointOffset=<1 0 0>
     }]
    relativeSpaceBone {boneName=root}
   }
   n38 = Anim Dynamics - Bone: dyn_Hair_extra_03(ComponentPose=@n37, Alpha=1, ExternalForce="0,0,0") {
    angularDampingOverride=0.9999
    angularSpringConstant=100
    bAngularSpring=true
    bGravityOverrideInSimSpace=true
    bOverrideAngularDamping=true
    simulationSpace=RootRelative
    alphaBoolBlend {blendOption=Linear}
    boundBone {boneName=dyn_Hair_extra_03}
    componentAppliedLinearAccClamp=<3500 3500 3500>
    componentLinearAccScale=<1.2 1.2 1.2>
    componentLinearVelScale=<1.2 1.2 1.2>
    physicsBodyDefinitions [{
      sphereCollisionRadius=0
      boundBone {boneName=dyn_Hair_extra_03}
      boxExtents=<20 10 10>
      constraintSetup {angularLimitsMax=<1 15 5>, angularLimitsMin=<-1 -15 -25>, angularTarget=<1 0 0>}
      localJointOffset=<1 0 0>
     }]
    relativeSpaceBone {boneName=root}
   }
   n39 = Anim Dynamics - Bone: dyn_Hair_extra_04(ComponentPose=@n38, Alpha=1, ExternalForce="0,0,0") {
    angularDampingOverride=0.9999
    angularSpringConstant=100
    bAngularSpring=true
    bGravityOverrideInSimSpace=true
    bOverrideAngularDamping=true
    simulationSpace=RootRelative
    alphaBoolBlend {blendOption=Linear}
    boundBone {boneName=dyn_Hair_extra_04}
    componentAppliedLinearAccClamp=<3500 3500 3500>
    componentLinearAccScale=<1.2 1.2 1.2>
    componentLinearVelScale=<1.2 1.2 1.2>
    physicsBodyDefinitions [{
      sphereCollisionRadius=0
      boundBone {boneName=dyn_Hair_extra_04}
      boxExtents=<20 10 10>
      constraintSetup {angularLimitsMax=<1 15 5>, angularLimitsMin=<-1 -15 -25>, angularTarget=<1 0 0>}
      localJointOffset=<1 0 0>
     }]
    relativeSpaceBone {boneName=root}
   }
   n40 = Anim Dynamics - Bone: dyn_Hair_extra_05(ComponentPose=@n39, Alpha=1, ExternalForce="0,0,0") {
    angularDampingOverride=0.9999
    angularSpringConstant=100
    bAngularSpring=true
    bGravityOverrideInSimSpace=true
    bOverrideAngularDamping=true
    simulationSpace=RootRelative
    alphaBoolBlend {blendOption=Linear}
    boundBone {boneName=dyn_Hair_extra_05}
    componentAppliedLinearAccClamp=<3500 3500 3500>
    componentLinearAccScale=<1.2 1.2 1.2>
    componentLinearVelScale=<1.2 1.2 1.2>
    physicsBodyDefinitions [{
      sphereCollisionRadius=0
      boundBone {boneName=dyn_Hair_extra_05}
      boxExtents=<20 10 10>
      constraintSetup {angularLimitsMax=<1 15 5>, angularLimitsMin=<-1 -15 -25>, angularTarget=<1 0 0>}
      localJointOffset=<1 0 0>
     }]
    relativeSpaceBone {boneName=root}
   }
   n41 = Anim Dynamics - Bone: dyn_Hair_extra_06(ComponentPose=@n40, Alpha=1, ExternalForce="0,0,0") {
    angularDampingOverride=0.9999
    angularSpringConstant=100
    bAngularSpring=true
    bGravityOverrideInSimSpace=true
    bOverrideAngularDamping=true
    simulationSpace=RootRelative
    alphaBoolBlend {blendOption=Linear}
    boundBone {boneName=dyn_Hair_extra_06}
    componentAppliedLinearAccClamp=<1500 1500 1500>
    componentLinearAccScale=<1.2 1.2 1.2>
    componentLinearVelScale=<1.2 1.2 1.2>
    physicsBodyDefinitions [{
      sphereCollisionRadius=0
      boundBone {boneName=dyn_Hair_extra_06}
      boxExtents=<20 12 12>
      constraintSetup {angularLimitsMax=<1 15 5>, angularLimitsMin=<-1 -15 -25>, angularTarget=<1 0 0>}
      localJointOffset=<1 0 0>
     }]
    relativeSpaceBone {boneName=root}
   }
   n42 = Anim Dynamics - Bone: dyn_Hair_extra_07(ComponentPose=@n41, Alpha=1, ExternalForce="0,0,0") {
    angularDampingOverride=0.9999
    angularSpringConstant=100
    bAngularSpring=true
    bGravityOverrideInSimSpace=true
    bOverrideAngularDamping=true
    simulationSpace=RootRelative
    alphaBoolBlend {blendOption=Linear}
    boundBone {boneName=dyn_Hair_extra_07}
    componentAppliedLinearAccClamp=<3500 3500 3500>
    componentLinearAccScale=<1.2 1.2 1.2>
    componentLinearVelScale=<1.2 1.2 1.2>
    physicsBodyDefinitions [{
      sphereCollisionRadius=0
      boundBone {boneName=dyn_Hair_extra_07}
      boxExtents=<20 10 10>
      constraintSetup {angularLimitsMax=<1 5 5>, angularLimitsMin=<-1 -30 -25>, angularTarget=<1 0 0>}
      localJointOffset=<1 0 0>
     }]
    relativeSpaceBone {boneName=root}
   }
   n43 = Anim Dynamics - Bone: dyn_Hair_extra_08(ComponentPose=@n42, Alpha=1, ExternalForce="0,0,0") {
    angularDampingOverride=0.9999
    angularSpringConstant=25
    bAngularSpring=true
    bGravityOverrideInSimSpace=true
    bOverrideAngularDamping=true
    simulationSpace=RootRelative
    alphaBoolBlend {blendOption=Linear}
    boundBone {boneName=dyn_Hair_extra_08}
    componentAppliedLinearAccClamp=<3500 3500 3500>
    componentLinearAccScale=<0.6 0.6 0.8>
    componentLinearVelScale=<0.6 0.6 0.8>
    physicsBodyDefinitions [{
      sphereCollisionRadius=0
      boundBone {boneName=dyn_Hair_extra_08}
      boxExtents=<20 15 15>
      constraintSetup {angularLimitsMax=<1 5 15>, angularLimitsMin=<-1 -35 -15>, angularTarget=<1 0 0>}
      localJointOffset=<1 0 0>
     }]
    relativeSpaceBone {boneName=root}
   }
   n44 = Anim Dynamics - Bone: dyn_Hair_extra_09(ComponentPose=@n43, Alpha=1, ExternalForce="0,0,0") {
    angularDampingOverride=0.9999
    angularSpringConstant=75
    bAngularSpring=true
    bGravityOverrideInSimSpace=true
    bOverrideAngularDamping=true
    simulationSpace=BoneRelative
    alphaBoolBlend {blendOption=Linear}
    boundBone {boneName=dyn_Hair_extra_09}
    componentAppliedLinearAccClamp=<1500 1500 1500>
    componentLinearAccScale=<1.2 1.2 1.2>
    componentLinearVelScale=<1.2 1.2 1.2>
    physicsBodyDefinitions [{
      sphereCollisionRadius=0
      boundBone {boneName=dyn_Hair_extra_09}
      boxExtents=<20 12 12>
      constraintSetup {angularLimitsMax=<1 5 5>, angularLimitsMin=<-1 -30 -25>, angularTarget=<1 0 0>}
      localJointOffset=<1 0 0>
     }]
    relativeSpaceBone {boneName=spine_05}
   }
   n45 = "Transform (Modify) Bone - Bone: dyn_mask"!(ComponentPose=@n44, bAlphaBoolEnabled:bool=Get mask() -> (mask:bool), Translation="0,0.05,0", Rotation="0,0,0", Scale="1,1,1") -> (Pose) {
    alphaInputType=Bool
    rotationSpace=BCS_BoneSpace
    translationMode=BMM_Additive
    translationSpace=BCS_BoneSpace
    alphaBoolBlend {blendOption=Linear}
    alphaScaleBiasClamp {interpSpeedIncreasing=2}
    boneToModify {boneName=dyn_mask}
    translation {y=0.05}
   }
   n46 = Control Rig!(Source=Component To Local(ComponentPose=@n45), Alpha, skydive:double=Get skydive() -> (skydive:double)) -> (Pose) {
    alphaBoolBlend {blendOption=Linear}
    controlRigAssetReference {blueprintRigClass=&/M_MED_LazarusLens_Style_FaceAcc_post_CtrlRig.M_MED_LazarusLens_Style_FaceAcc_post_CtrlRig_C}
   }
   Output Pose(Result=@n46) {layerGroup=None}
  ]}
 EventGraph {nodes [
   Event BlueprintUpdateAnimation()
   > Set mask(mask:bool=true) -> (Output_Get:bool)
   > n1 = Gravity Override State Function DEPRECATED(
    CharacterPartAnimInstance=Self-Reference()
    GameTimeInSeconds=GetGameTimeInSeconds()
    GravityOverrideParams=Get GravityOverrideParams() -> (GravityOverrideParams:Gravity Override Params Struct Structure)
   )
   > Set GravityOverride(GravityOverride=@n1)
   > n2 = Branch(Condition=Get bIsSkyDiving())
   > Set skydive(skydive:double=0.7) -> (Output_Get:double)
   Set skydive(execute=@n2.else, skydive=0)
   n3 = Event OnBeginFalling()
   n4 = Branch(execute=[@n3.then, @n6.then], Condition=CheckForBackbling())
   Set jumping(execute=[@n5.then, @n7.then, @n4.then], jumping:double=0) -> (Output_Get:double)
   Set jumping(execute=@n4.else, jumping=1)
   n5 = Event OnEndFalling()
   n6 = Event OnBeginJumping()
   n7 = Event OnEndJumping()
   Event BlueprintInitializeAnimation()
   > Set Skel_Mesh(Skel_Mesh=GetOwnerSkeletalMeshComponent())
  ]}
]
variables [
 GravityOverride {type=Vector}
 GravityOverrideParams {type=GravityOverrideParamsStruct}
 jumping {type=number}
 mask {type=boolean}
 Skel_Mesh {type=SkeletalMeshComponent}
 skydive {type=number}
]
